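# Niagara idle head effect for the female medium RankedTrooper cosmetic, drawing red eye glow and glint sprites with lights.
# FN_EyeGlow bursts two particles tracking skeletal mesh sockets or bones, while FN_EyeGlint spawns short-lived glare flashes gated by mask and state-change flags.
# Dot products of camera vector and sampled socket orientation drive view-dependent alpha and camera offset.
# Mask and head-hidden flags set visibility tags; front-end modifiers adjust glint size and opacity.
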NiagaraSystem /BRCosmetics/Characters/Player/Female/Medium/Bodies/F_MED_RankedTrooper/FX/NS_F_MED_RankedTrooper_HeadIdle
dependencies {
 usedDataInterfaces [&NiagaraDataInterfaceCamera, &NiagaraDataInterfaceCurve, &NiagaraDataInterfaceVector2DCurve]
 usedDynamicInputs [&SetBoolByIntComparison]
 usedModules [
  &self:SystemSpawnScript.SystemScriptSource.SystemScriptGraph.NiagaraNodeAssignment_0.SetVariables_#1
  &self:SystemSpawnScript.SystemScriptSource.SystemScriptGraph.NiagaraNodeAssignment_2.SetVariables_#2
  &self:SystemSpawnScript.SystemScriptSource.SystemScriptGraph.NiagaraNodeAssignment_1.SetVariables_#3
  &self:SystemSpawnScript.SystemScriptSource.SystemScriptGraph.NiagaraNodeAssignment_3.SetVariables_#4
  &SystemState
  /Game/Effects/Niagara/Modules/Logic/UpdateValueStateChange
 ]
}
emitters {
 FN_EyeGlint {
  emitterUpdateScript [
   &EmitterState()
   &SpawnRate(SpawnRate=10, Use Spawn Probability=false) {enabled=false}
   &SpawnPerFrame(
    Spawn:NiagaraBool=&InvertBool(Boolean:NiagaraBool=$System.UpdateValueStateChange.Changed)
    Spawn Count:NiagaraInt32=&MakeCustomIntFromBool(Bool=$System.IsWearingMask, True INT=1, False INT=0)
    Use Spawn Probability=false
   )
   &SpawnBurst_Instantaneous(Spawn Count=2, Use Spawn Probability=false) {enabled=false}
  ]
  particleSpawnScript [/Niagara/Modules/Spawn/Initialization/InitializeParticle(
    Color=&MultiplyLinearColor(
     Linear Color=<rgba 3 0.1406 0.1406 1>
     Scale Factor=&MakeFloat4FromVectorAndFloat("Vector (RGB)"=<1 1 1>, "Float (Alpha)"=&FN_FrontEndFloatModifier(Gameplay=0.3, Front End=0.4))
    )
    Lifetime=1.5
    Mesh Scale=<1.3 1.3 1.3>
    Module.Write Color=true
    Module.Write Lifetime=true
    Module.Write Mass=true
    Module.Write Material Random=true
    Module.Write Position=false
    Module.Write Scale=false
    Module.Write SpriteRotation=false
    Module.Write SpriteSize=true
    Position=&SimulationPosition()
    Sprite Size=&Vector2DFromFloat(Value=&FN_FrontEndFloatModifier(Gameplay=20, Front End=16))
   )]
  particleUpdateScript [
   &ParticleState()
   &SpriteRotationRate(Rotation Rate=&FloatFromCurve(
     FloatCurve=DataInterface() {propertyValues {Curve {keys [<at 0 1 RCIM_Cubic>, {arriveTangent=-1.4023, interpMode=RCIM_Cubic, leaveTangent=-1.4023, tangentMode=RCTM_User, time=1, value=0.1}]}}}
     Scale Curve=-250
    ))
   &ScaleSpriteSize(
    Uniform Curve Scale=1.1
    Uniform Curve Sprite Scale=DataInterface() {propertyValues {Curve {keys [
        <at 0 0 RCIM_Cubic>
        {arriveTangent=-1.7771, interpMode=RCIM_Cubic, leaveTangent=-1.7771, tangentMode=RCTM_User, time=0.1, value=1.625}
        {tangentMode=RCTM_User, time=1, value=0.85}
       ]}}}
   )
   &SkeletalMeshLocation(
    Apply Sampled Orientation to Mesh Particles=true
    Apply To Particle Position=false
    Bone / Socket Index=$System.BoneLODOffset
    Inherit Velocity=false
    Interpolate Mesh Sampling=false
    MaskSpawnGroup=false
    Mesh Sampling Type=::NewEnumerator5
    Sampled Orientation Coordinate Space=::World
    Sampled Position Offset=<0 0.1286 0>
    Sampled Position Offset Coordinate Space:ENiagaraMeshTransforms=::NewEnumerator3
    Skeletal Mesh=$System.Skeletal Mesh
    Socket Sampling Mode=::NewEnumerator5
   ) {enabled=false}
   &SkeletalMeshLocation(
    Apply Sampled Orientation to Mesh Particles=true
    Inherit Velocity=false
    Interpolate Mesh Sampling=false
    MaskSpawnGroup=false
    Mesh Sampling Type=::NewEnumerator5
    Sampled Orientation Coordinate Space=::World
    Sampled Position Offset=<0 0.1286 0>
    Sampled Position Offset Coordinate Space:ENiagaraMeshTransforms=::NewEnumerator3
    Skeletal Mesh=$System.Skeletal Mesh
    Socket Sampling Mode=::NewEnumerator5
   ) {module=SkeletalMeshLocation001}
   /SetVariables(Particles.Position=/Niagara/DynamicInputs/TypeConversions/MakeCustomPositionFromBool(
     Bool:NiagaraBool=$System.IsWearingMask
     True Position:NiagaraPosition=$Output.SkeletalMeshLocation001.SampledPosition
     False Position:NiagaraPosition=$Output.SkeletalMeshLocation.SampledPosition
    )) {isSetParameters=true, module=SetVariables_#5}
   /SetVariables(Particles.VisWeight:NiagaraFloat=&DotProduct(
     Vector A=&ReturnCameraVector(Camera Vectors=::NewEnumerator1, Coordinate Space=::World)
     Vector B=&RotateVectorByQuaternion(
      Vector To Rotate=<0 -1 0>
      Quaternion=/Niagara/DynamicInputs/TypeConversions/MakeCustomQuaternionFromBool(
       Bool:NiagaraBool=$System.IsWearingMask
       True Quat:Quat4f=$Output.SkeletalMeshLocation001.SampledRotation
       False Quat:Quat4f=$Output.SkeletalMeshLocation.SampledRotation
      )
     )
    )) {isSetParameters=true, module=SetVariables_#6}
   /SetVariables(Particles.VisibilityDot:NiagaraFloat=$Particles.VisWeight) {isSetParameters=true}
   /SetVariables(Particles.GlowVizWeight:NiagaraFloat=&RemapRange(Input Value To Remap=$Particles.VisibilityDot, Input Min=0.125, Input Max=0.5, Output Min=0, Output Max=1)) {isSetParameters=true}
   /SetVariables(Particles.VisWeight:NiagaraFloat=&RemapRange(Input Value To Remap=$Particles.VisWeight, Input Min=-0.25, Input Max=0.25, Output Min=-1, Output Max=1)) {isSetParameters=true, module=SetVariables_#7}
   &CameraOffset(Camera Offset Amount=&Multiply_Float(
     A=&FloatFromCurve(
      FloatCurve=DataInterface() {propertyValues {Curve {keys [<at 0 0>,<at 0.1 1>]}}}
      Scale Curve=&MakeCustomFloatFromBool(Bool=$System.IsWearingMask, True Float=10, False Float=10)
     )
     B=$Particles.VisWeight
    ))
   &ScaleColor(
    Scale Alpha=&FloatFromCurve(FloatCurve=DataInterface() {propertyValues {Curve {keys [<at 0 1 RCIM_Cubic>, {arriveTangent=-1, interpMode=RCIM_Cubic, leaveTangent=-1, time=0.5, value=0.2}, <at 1 0 RCIM_Cubic>]}}})
    Scale RGB=&VectorFromFloat(Value=&FloatFromCurve(
      FloatCurve=DataInterface() {propertyValues {Curve {keys [
          {arriveTangent=9.0725, interpMode=RCIM_Cubic, leaveTangent=9.0725, tangentMode=RCTM_User, time=0, value=0}
          {arriveTangent=-2.5249, interpMode=RCIM_Cubic, leaveTangent=-2.5249, tangentMode=RCTM_User, time=0.1408, value=0.9847}
          {arriveTangent=-0.0405, interpMode=RCIM_Cubic, leaveTangent=-0.0405, tangentMode=RCTM_User, time=1, value=0}
         ]}}}
      Scale Curve=2
     ))
    ScaleA=true
    ScaleRGB=true
   )
   /SetVariables(
    Particles.GlowColor=&MultiplyLinearColor(
     Linear Color=$Particles.Color
     Scale Factor=&Multiply_Vector4(
      A=<0.5 0.5 0.5 0.15>
      B=&MakeFloat4FromVectorAndFloat("Vector (RGB)"=<1 1 1>, "Float (Alpha)"=$Particles.GlowVizWeight)
     )
    )
    Particles.GlowSize=&Multiply_Vector2D_ByFloat(Float=3, Vector2D=$Particles.SpriteSize)
   ) {isSetParameters=true}
   /SetVariables(
    Particles.GlareColor:LinearColor=&MultiplyLinearColor(
     Linear Color=&MultiplyLinearColor(
      Linear Color=$Particles.Color
      Scale Factor=&MakeVector4_FromLinearColor(LinearColor=<rgba 5 0 0.0427 0.02>)
     )
     Scale Factor=&MakeFloat4FromVectorAndFloat("Vector (RGB)"=<1 1 1>, "Float (Alpha)"=$Particles.GlowVizWeight)
    )
    Particles.GlareSize:Vector2f=&Multiply_Vector2D(
     A=<12 18.4797>
     B=&Vector2DFromCurve(
      Vector2Curve=DataInterface() {propertyValues {
        XCurve {keys [
          {arriveTangent=-0.25, interpMode=RCIM_Cubic, time=0.1, value=0}
          {arriveTangent=2, interpMode=RCIM_Cubic, leaveTangent=2, time=0.3, value=2.5}
          <at 1.1 2 RCIM_Cubic>
         ]}
        YCurve {keys [<at 0 1>]}
       }}
      ScaleCurve=<2 1>
     )
    )
   ) {isSetParameters=true}
   &ScaleColor(
    Scale Alpha=&FloatFromCurve(
     FloatCurve=DataInterface() {propertyValues {Curve {keys [
         <at 0 0 RCIM_Cubic>
         {arriveTangent=1.1562, interpMode=RCIM_Cubic, leaveTangent=1.1562, time=0.1, value=1}
         {arriveTangent=-1.1111, interpMode=RCIM_Cubic, leaveTangent=-1.1111, time=0.8, value=0.925}
         <at 1 0 RCIM_Cubic>
        ]}}}
     Scale Curve=&MakeCustomFloatFromBool(Bool=$System.IsWearingMask, True Float=1, False Float=1)
    )
    ScaleA=true
    ScaleRGB=true
   ) {module=ScaleColor001}
   &FN_Light_Attributes(
    Color=$Particles.Color
    Create Color Attribute=true
    Create Position Attribute=true
    Create Radius Attribute=true
    Enable Particle Light Enable=true
    Position=&AddVectorToPosition(Position=/Niagara/DynamicInputs/Helpers/SimulationPosition(), Vector=<0 3 0>)
    Radius=&FloatFromCurve(FloatCurve=DataInterface() {propertyValues {Curve {keys [<at 0 1>,<at 1 0>]}}}, Scale Curve=10)
    Scale Color=&FloatFromCurve(FloatCurve=DataInterface() {propertyValues {Curve {keys [<at 0 1>,<at 1 0>]}}}, Scale Curve=0)
   )
   /SetVariables(Particles.VisibilityTag=&MakeCustomIntFromBool(
     Bool=$System.IsWearingMask
     True INT=&MakeCustomIntFromBool(Bool=$User.bisHeadHidden, True INT=1, False INT=0)
     False INT=1
    )) {isSetParameters=true}
  ]
  rendererClasses [&NiagaraSpriteRendererProperties, &NiagaraLightRendererProperties]
  renderers [
   {rendererClass=&NiagaraSpriteRendererProperties}
   {rendererClass=&NiagaraLightRendererProperties, rendererIndex=1}
   {rendererClass=&NiagaraSpriteRendererProperties, rendererIndex=2}
   {rendererClass=&NiagaraSpriteRendererProperties, rendererIndex=3}
  ]
 }
 FN_EyeGlow {
  emitterUpdateScript [
   &EmitterState(
    Inactive Response=::NewEnumerator1
    Life Cycle Mode=::NewEnumerator1
    Loop Behavior=::NewEnumerator1
    Loop Duration Mode=::NewEnumerator1
    UseLoopDelay=false
   )
   &SpawnBurst_Instantaneous(Spawn Count=2, Use Spawn Probability=false)
  ]
  particleSpawnScript [
   &InitializeParticle(
    Color=&MultiplyLinearColor(
     Linear Color=<rgba 3 0 0 0.5>
     Scale Factor=&MakeFloat4FromVectorAndFloat("Vector (RGB)"=<1 1 1>, "Float (Alpha)"=&FN_FrontEndFloatModifier(Gameplay=0.2, Front End=0.2))
    )
    Mesh Scale=<1.3 1.3 1.3>
    Module.Write Color=true
    Module.Write Lifetime=true
    Module.Write Mass=true
    Module.Write Material Random=true
    Module.Write Position=false
    Module.Write Scale=false
    Module.Write SpriteRotation=false
    Module.Write SpriteSize=true
    Position=/Niagara/Functions/Localspace/SimulationPosition()
    Sprite Size=&Vector2DFromFloat(Value=20)
   )
   &PartitionParticles(Particles Per Partition=1, Partition Mode:ENiagara_PartitionMode=::NewEnumerator1)
  ]
  particleUpdateScript [
   &ParticleState(Kill Particles When Lifetime Has Elapsed=false, Let Infinitely Lived Particles Die When Emitter Deactivates=true)
   /SetVariables(Particles.VisibilityTag=&MakeCustomIntFromBool(Bool=$System.IsWearingMask, True INT=0, False INT=1)) {isSetParameters=true}
   &SkeletalMeshLocation(
    Apply Sampled Orientation to Mesh Particles=true
    Apply To Particle Position=false
    Bone / Socket Index=&MakeCustomIntFromBool(
     Bool=$System.IsWearingMask
     True INT=1
     False INT=&Add_Integer(A=$Particles.PartitionParticles.Partition, B=$System.BoneLODOffset)
    )
    Inherit Velocity=false
    Interpolate Mesh Sampling=false
    MaskSpawnGroup=false
    Mesh Sampling Type=::NewEnumerator5
    Sampled Orientation Coordinate Space=::World
    Sampled Position Offset=<0 0.1286 0>
    Sampled Position Offset Coordinate Space:ENiagaraMeshTransforms=::NewEnumerator3
    Skeletal Mesh=$System.Skeletal Mesh
    Socket Sampling Mode=::NewEnumerator5
   ) {enabled=false}
   &SkeletalMeshLocation(
    Apply Sampled Orientation to Mesh Particles=true
    Bone / Socket Index=$Particles.PartitionParticles.Partition
    Inherit Velocity=false
    Interpolate Mesh Sampling=false
    MaskSpawnGroup=false
    Mesh Sampling Type=::NewEnumerator5
    Sampled Orientation Coordinate Space=::World
    Sampled Position Offset=<0 0.1286 0>
    Sampled Position Offset Coordinate Space:ENiagaraMeshTransforms=::NewEnumerator3
    Skeletal Mesh=$System.Skeletal Mesh
    Socket Sampling Mode=::NewEnumerator5
   ) {module=SkeletalMeshLocation001}
   /SetVariables(Particles.Position=&MakeCustomPositionFromBool(
     Bool:NiagaraBool=$System.IsWearingMask
     True Position:NiagaraPosition=$Output.SkeletalMeshLocation001.SampledPosition
     False Position:NiagaraPosition=$Output.SkeletalMeshLocation.SampledPosition
    )) {enabled=false, isSetParameters=true, module=SetVariables_#5}
   /SetVariables(Particles.VisWeight:NiagaraFloat=&DotProduct(
     Vector A=&ReturnCameraVector(Camera Vectors=::NewEnumerator1, Coordinate Space=::World)
     Vector B=&RotateVectorByQuaternion(
      Vector To Rotate=<0 -1 0>
      Quaternion=&MakeCustomQuaternionFromBool(
       Bool:NiagaraBool=$System.IsWearingMask
       True Quat:Quat4f=$Output.SkeletalMeshLocation001.SampledRotation
       False Quat:Quat4f=$Output.SkeletalMeshLocation.SampledRotation
      )
     )
    )) {isSetParameters=true, module=SetVariables_#6}
   /SetVariables(Particles.VisWeight:NiagaraFloat=&RemapRange(Input Value To Remap=$Particles.VisWeight, Input Min=0.125, Input Max=0.5, Output Min=0, Output Max=1)) {isSetParameters=true, module=SetVariables_#7}
   &CameraOffset(Camera Offset Amount=&MakeCustomFloatFromBool(Bool=$System.IsWearingMask, True Float=15, False Float=3))
   &ScaleColor(
    Scale Alpha=&Multiply_Float(
     A=$Particles.VisWeight
     B=&FloatFromCurve(
      FloatCurve=DataInterface() {propertyValues {Curve {keys [
          {arriveTangent=4.1785, interpMode=RCIM_Cubic, time=0, value=0}
          {arriveTangent=0.2158, interpMode=RCIM_Cubic, leaveTangent=0.2158, tangentMode=RCTM_User, time=0.3, value=0.1}
          {interpMode=RCIM_Cubic, leaveTangent=0.0224, time=1, value=1}
         ]}}}
      CurveIndex=$System.Age
     )
    )
    ScaleA=true
    ScaleRGB=true
   )
   &DynamicMaterialParameters(
    Index 0 Param 1=$Particles.VisWeight
    Module.Param0WriteEnabled=true
    Module.Param1WriteEnabled=true
    Module.Param2WriteEnabled=true
    Module.Param3WriteEnabled=true
   )
   /SetVariables(Particles.VisibilityTag=&MakeCustomIntFromBool(Bool=$User.bisHeadHidden, True INT=1, False INT=$Particles.VisibilityTag)) {isSetParameters=true}
  ]
  rendererClasses [&NiagaraSpriteRendererProperties, &NiagaraLightRendererProperties]
  renderers [
   {rendererClass=&NiagaraSpriteRendererProperties}
   {rendererClass=&NiagaraLightRendererProperties, rendererIndex=1}
  ]
 }
}
summary {
 systemName=NS_F_MED_RankedTrooper_HeadIdle
 emitters [
  {emitterName=FN_EyeGlow, rendererClasses [&NiagaraSpriteRendererProperties, &NiagaraLightRendererProperties]}
  {emitterName=FN_EyeGlint, rendererClasses [&NiagaraSpriteRendererProperties, &NiagaraLightRendererProperties]}
 ]
 userVariables [
  User.bIsFrontEnd {type=NiagaraBool}
  User.bIsFrontEndPreview {type=NiagaraBool}
  User.bisHeadHidden {type=NiagaraBool}
  User.KillCount {type=NiagaraInt32}
  User.MaskIndex {type=NiagaraInt32}
  User.MaskMesh {
   type=NiagaraDataInterfaceSkeletalMesh
   defaultValue {
    struct=NiagaraExt_VariableValue_DataInterface
    value {dataInterface=&self:NiagaraDataInterfaceSkeletalMesh_1}
   }
  }
  User.SkeletalMeshCurrentLOD {type=NiagaraInt32}
  User.StyleIndex {type=NiagaraInt32}
 ]
}
