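# Physics asset for the F_MED_RankedTrooper female body cosmetic, built around a dangling belt strap on the character.
# A chain of four strap capsule bodies hangs from a zero-size master body under the pelvis, each joint with tight swing and twist limit angles.
# Capsules on the pelvis, left thigh and left calf cover only that side of the lower body, joined by free angular constraints.
# The last strap segment carries two much larger overlapping capsules.
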PhysicsAsset /BRCosmetics/Characters/Player/Female/Medium/Bodies/F_MED_RankedTrooper/Meshes/F_MED_RankedTrooper_Physics
bodies {
 calf_l [{shapeType=Capsule, length=32, radius=6, shapeName=calf_l_capsule, center=<-20 -2.5444 0>, rotation=<rot 0 90 90>}]
 dyn_BeltStrap_01_jnt [{shapeType=Capsule, length=4, radius=1, shapeName=dyn_BeltStrap_01_jnt_capsule, center=<1.5 0 0>, rotation=<rot 0 0 90>}]
 dyn_BeltStrap_02_jnt [{shapeType=Capsule, length=4, radius=1, shapeName=dyn_BeltStrap_02_jnt_capsule, center=<1.5 0 0>, rotation=<rot 0 0 90>}]
 dyn_BeltStrap_03_jnt [{shapeType=Capsule, length=4, radius=1.25, shapeName=dyn_BeltStrap_03_jnt_capsule, center=<3 0 0>, rotation=<rot 0 0 90>}]
 dyn_BeltStrap_04_jnt [
  {shapeType=Capsule, length=12.5, radius=8, shapeName=dyn_BeltStrap_04_jnt_capsule, center=<14 0 4>, rotation=<rot 0 90.0002 75>}
  {shapeType=Capsule, length=12.5, radius=8, shapeName=dyn_BeltStrap_04_jnt_capsule, center=<16.1197 0 3.5>, rotation=<rot 0 0 90>}
 ]
 dyn_BeltStrap_MSTR [{shapeType=Capsule, length=0, radius=0, shapeName=dyn_BeltStrap_MSTR_capsule}]
 pelvis [{shapeType=Capsule, length=10, radius=10, shapeName=pelvis_capsule, center=<-2.3445 -2.4055 0>, rotation=<rot 0 90 0>}]
 thigh_l [{shapeType=Capsule, length=32, radius=7.5, shapeName=thigh_l_capsule, center=<-20 0 0>, rotation=<rot 0 90 90>}]
}
constraints [
 {
  bone1Name=dyn_BeltStrap_MSTR
  bone2Name=pelvis
  swing1LimitDegrees=45
  swing1Motion=Free
  swing2LimitDegrees=45
  swing2Motion=Free
  twistLimitDegrees=45
  twistMotion=Free
 }
 {
  bone1Name=thigh_l
  bone2Name=pelvis
  swing1LimitDegrees=45
  swing1Motion=Free
  swing2LimitDegrees=45
  swing2Motion=Free
  twistLimitDegrees=45
  twistMotion=Free
 }
 {bone1Name=dyn_BeltStrap_01_jnt, bone2Name=dyn_BeltStrap_MSTR, swing1LimitDegrees=15, swing2LimitDegrees=25, twistLimitDegrees=5}
 {bone1Name=dyn_BeltStrap_02_jnt, bone2Name=dyn_BeltStrap_01_jnt, swing1LimitDegrees=15, swing2LimitDegrees=25, twistLimitDegrees=5}
 {bone1Name=dyn_BeltStrap_03_jnt, bone2Name=dyn_BeltStrap_02_jnt, swing1LimitDegrees=15, swing2LimitDegrees=25, twistLimitDegrees=5}
 {bone1Name=dyn_BeltStrap_04_jnt, bone2Name=dyn_BeltStrap_03_jnt, swing1LimitDegrees=15, swing2LimitDegrees=25, twistLimitDegrees=5}
 {
  bone1Name=calf_l
  bone2Name=thigh_l
  swing1LimitDegrees=45
  swing1Motion=Free
  swing2LimitDegrees=45
  swing2Motion=Free
  twistLimitDegrees=45
  twistMotion=Free
 }
]
