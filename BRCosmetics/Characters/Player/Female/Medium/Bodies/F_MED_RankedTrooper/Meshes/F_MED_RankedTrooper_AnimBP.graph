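# Post-processing AnimBlueprint for the F_MED_RankedTrooper female body cosmetic, adding secondary motion to straps and adjusting attachments.
# The parent mesh's copied pose feeds a Control Rig, belt-strap constraint and offsets, pelvis-space RigidBody, calf-strap AnimDynamics and backpack offsets.
# A front-end idle curve weights a DefaultSlot blend from the gun holster bone.
# The EventGraph updates gravity override via a deprecated function, plus skydive and front-end variables.
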AnimBlueprint /BRCosmetics/Characters/Player/Female/Medium/Bodies/F_MED_RankedTrooper/Meshes/F_MED_RankedTrooper_AnimBP
target_skeleton=&/F_MED_RankedTrooper_Skeleton
graphs [
 AnimGraph {nodes [
   n1 = Control Rig(Source=Copy Pose From Mesh() {bUseAttachedParent=true}) {
    alphaBoolBlend {blendOption=Linear}
    controlRigAssetReference {blueprintRigClass=&/F_MED_RankedTrooper_CtrlRig.F_MED_RankedTrooper_CtrlRig_C}
   }
   n2 = "Constraint - dyn_BeltStrap_02_jnt (PP)"!(ComponentPose=Local To Component(LocalPose=@n1), Alpha=1, ConstraintWeights_0=0.5, ConstraintWeights_1:float=0.5) -> (Pose) {
    alphaBoolBlend {blendOption=Linear}
    boneToModify {boneName=dyn_BeltStrap_02_jnt}
    constraintSetup [{targetBone {boneName=thigh_l}}, {targetBone {boneName=thigh_twist_01_l}}]
    constraintWeights [1,1]
   }
   n3 = "Transform (Modify) Bone - Bone: dyn_BeltStrap_MSTR"(ComponentPose=@n2, Alpha=1, Translation="0,0,0", Rotation="-5,0,0", Scale="1,1,1") {
    rotationMode=BMM_Additive
    rotationSpace=BCS_BoneSpace
    translationMode=BMM_Additive
    translationSpace=BCS_BoneSpace
    alphaBoolBlend {blendOption=Linear}
    alphaScaleBiasClamp {bInterpResult=true, interpSpeedDecreasing=4, interpSpeedIncreasing=6}
    boneToModify {boneName=dyn_BeltStrap_MSTR}
    rotation {pitch=-5}
   }
   n4 = RigidBody!(ComponentPose=@n3, Alpha, OverrideWorldGravity:Vector=Get GravityOverride(), ExternalForce) -> (Pose) {
    bClampLinearTranslationLimitToRefPose=true
    bOverrideWorldGravity=true
    overridePhysicsAsset=&/F_MED_RankedTrooper_Physics
    simulationSpace=BaseBoneSpace
    alphaBoolBlend {blendOption=Linear}
    alphaScaleBiasClamp {interpSpeedIncreasing=2}
    baseBoneRef {boneName=pelvis}
    componentAppliedLinearAccClamp=<3000 3000 3000>
    componentLinearAccScale=<0.3 0.3 0.3>
    componentLinearVelScale=<0.3 0.3 0.3>
    simSpaceSettings {
     maxAngularAcceleration=1000
     maxAngularVelocity=3
     maxLinearAcceleration=700
     maxLinearVelocity=500
     velocityScaleZ=0.25
     worldAlpha=0.5
     externalAngularVelocity {z=0.1}
     externalLinearDragV=<0.2 0.2>
    }
   }
   n5 = "Transform (Modify) Bone - Bone: dyn_BeltStrap_MSTR"(ComponentPose=@n4, Alpha=1, Translation="-2,0,1", Rotation="0,0,0", Scale="1,1,1") {
    rotationMode=BMM_Additive
    rotationSpace=BCS_BoneSpace
    translationMode=BMM_Additive
    translationSpace=BCS_BoneSpace
    alphaBoolBlend {blendOption=Linear}
    alphaScaleBiasClamp {bInterpResult=true, interpSpeedDecreasing=4, interpSpeedIncreasing=6}
    boneToModify {boneName=dyn_BeltStrap_MSTR}
    translation {x=-2, z=1}
   }
   n6 = Anim Dynamics - Bone: dyn_calf_l_upperstrap_jnt(ComponentPose=@n5, Alpha=1, ExternalForce="0,0,0") {
    angularDampingOverride=0.9999
    angularSpringConstant=200
    bAngularSpring=true
    bOverrideAngularDamping=true
    bUseGravityOverride=true
    bUsePlanarLimit=false
    gravityScale=0.35
    simulationSpace=RootRelative
    alphaBoolBlend {blendOption=Linear}
    boundBone {boneName=dyn_calf_l_upperstrap_jnt}
    componentAppliedLinearAccClamp=<6000 6000 6000>
    componentLinearAccScale=<0.5 0.7 0.7>
    componentLinearVelScale=<0.5 0.7 0.7>
    gravityOverride {z=-980}
    physicsBodyDefinitions [{
      boundBone {boneName=dyn_calf_l_upperstrap_jnt}
      boxExtents=<20 30 30>
      constraintSetup {angularLimitsMax=<0 15 15>, angularLimitsMin=<0 -15 -15>, angularTarget=<1 0 0>}
      localJointOffset=<6 0 0>
     }]
   }
   n7 = Anim Dynamics - Bone: dyn_calf_r_upperstrap_jnt(ComponentPose=@n6, Alpha=1, ExternalForce="0,0,0") {
    angularDampingOverride=0.9999
    angularSpringConstant=200
    bAngularSpring=true
    bOverrideAngularDamping=true
    bUseGravityOverride=true
    bUsePlanarLimit=false
    gravityScale=0.35
    simulationSpace=RootRelative
    alphaBoolBlend {blendOption=Linear}
    boundBone {boneName=dyn_calf_r_upperstrap_jnt}
    componentAppliedLinearAccClamp=<6000 6000 6000>
    componentLinearAccScale=<0.5 0.7 0.7>
    componentLinearVelScale=<0.5 0.7 0.7>
    gravityOverride {z=-980}
    physicsBodyDefinitions [{
      boundBone {boneName=dyn_calf_r_upperstrap_jnt}
      boxExtents=<20 30 30>
      constraintSetup {angularLimitsMax=<0 15 15>, angularLimitsMin=<0 -15 -15>, angularTarget=<1 0 0>}
      localJointOffset=<6 0 0>
     }]
   }
   n8 = Anim Dynamics - Bone: dyn_calf_l_lowerstrap_jnt(ComponentPose=@n7, Alpha=1, ExternalForce="0,0,0") {
    angularDampingOverride=0.9999
    angularSpringConstant=200
    bAngularSpring=true
    bOverrideAngularDamping=true
    bUseGravityOverride=true
    bUsePlanarLimit=false
    gravityScale=0.35
    simulationSpace=RootRelative
    alphaBoolBlend {blendOption=Linear}
    boundBone {boneName=dyn_calf_l_lowerstrap_jnt}
    componentAppliedLinearAccClamp=<6000 6000 6000>
    componentLinearAccScale=<0.5 0.7 0.7>
    componentLinearVelScale=<0.5 0.7 0.7>
    gravityOverride {z=-980}
    physicsBodyDefinitions [{
      boundBone {boneName=dyn_calf_l_lowerstrap_jnt}
      boxExtents=<20 30 30>
      constraintSetup {angularLimitsMax=<0 15 15>, angularLimitsMin=<0 -15 -15>, angularTarget=<1 0 0>}
      localJointOffset=<6 0 0>
     }]
   }
   n9 = Anim Dynamics - Bone: dyn_calf_r_lowerstrap_jnt(ComponentPose=@n8, Alpha=1, ExternalForce="0,0,0") {
    angularDampingOverride=0.9999
    angularSpringConstant=200
    bAngularSpring=true
    bOverrideAngularDamping=true
    bUseGravityOverride=true
    bUsePlanarLimit=false
    gravityScale=0.35
    simulationSpace=RootRelative
    alphaBoolBlend {blendOption=Linear}
    boundBone {boneName=dyn_calf_r_lowerstrap_jnt}
    componentAppliedLinearAccClamp=<6000 6000 6000>
    componentLinearAccScale=<0.5 0.7 0.7>
    componentLinearVelScale=<0.5 0.7 0.7>
    gravityOverride {z=-980}
    physicsBodyDefinitions [{
      boundBone {boneName=dyn_calf_r_lowerstrap_jnt}
      boxExtents=<20 30 30>
      constraintSetup {angularLimitsMax=<0 15 15>, angularLimitsMin=<0 -15 -15>, angularTarget=<1 0 0>}
      localJointOffset=<6 0 0>
     }]
   }
   n10 = "Transform (Modify) Bone - Bone: attach_backpack"(ComponentPose=@n9, Alpha=1, Translation="0,-2.5,0", Rotation="0,0,0", Scale="1,1,1") {
    rotationMode=BMM_Additive
    rotationSpace=BCS_BoneSpace
    translationMode=BMM_Additive
    translationSpace=BCS_BoneSpace
    alphaBoolBlend {blendOption=Linear}
    alphaScaleBiasClamp {bInterpResult=true, interpSpeedDecreasing=4, interpSpeedIncreasing=4}
    boneToModify {boneName=attach_backpack}
    translation {y=-2.5}
   }
   n11 = "Transform (Modify) Bone - Bone: attach_backpack"(ComponentPose=@n10, Alpha=1, Translation="0,0,0", Rotation="0,0,10", Scale="1,1,1") {
    rotationMode=BMM_Additive
    rotationSpace=BCS_BoneSpace
    translationMode=BMM_Additive
    translationSpace=BCS_BoneSpace
    alphaBoolBlend {blendOption=Linear}
    alphaScaleBiasClamp {bInterpResult=true, interpSpeedDecreasing=4, interpSpeedIncreasing=6}
    boneToModify {boneName=attach_backpack}
    rotation {roll=10}
   }
   n12 = "Transform (Modify) Bone - Bone: attach_backpack"(ComponentPose=@n11, Alpha=1, Translation="0,0,0", Rotation="0,0,10", Scale="1,1,1") {
    rotationMode=BMM_Additive
    rotationSpace=BCS_BoneSpace
    translationMode=BMM_Additive
    translationSpace=BCS_BoneSpace
    alphaBoolBlend {blendOption=Linear}
    alphaScaleBiasClamp {bInterpResult=true, interpSpeedDecreasing=4, interpSpeedIncreasing=6}
    boneToModify {boneName=attach_backpack}
    rotation {roll=10}
   }
   Output Pose(Result=Layered blend per bone(
     BasePose=Component To Local(ComponentPose=@n12)
     BlendPoses_0=Slot 'DefaultSlot'()
     BlendWeights_0=Get CustomFELobbyIdleFloat() -> (CustomFELobbyIdleFloat:float)
    ) {layerSetup [{branchFilters [{boneName=dyn_gunHolster_Base_MSTR}]}]}) {layerGroup=None}
  ]}
 EventGraph {nodes [
   Event BlueprintUpdateAnimation()
   > n1 = Sequence!(execute) -> (then_0, then_1, then_2:Exec, then_3:Exec)
   > n2 = Gravity Override State Function DEPRECATED(CharacterPartAnimInstance=Self-Reference(), GameTimeInSeconds=GetGameTimeInSeconds(), GravityOverrideParams=Get Gravity Override Params())
   > Set GravityOverride(GravityOverride=@n2)
   Set Is Skydive Floating Alpha(execute=@n1.then_1, Is Skydive Floating Alpha:double="To Float (Boolean)"(InBool=Get bIsSkydiveFloating())) -> (Output_Get:double)
   n3 = Get bRigidBodyEnabled()
   > Set FrontEnd_RigidBodiesDisabled(FrontEnd_RigidBodiesDisabled:bool=AND Boolean(A=Get bIsInFrontEnd(), B=NOT Boolean(A=@n3))) -> (Output_Get:bool)
   n4 = Branch(execute=@n1.then_2, Condition=Get bIsSkyDiving())
   > Set BeltMask Alpha(BeltMask Alpha:double=0.5) -> (Output_Get:double)
   Set BeltMask Alpha(execute=@n4.else, BeltMask Alpha=1)
   Branch(execute=@n1.then_3, Condition=Get bIsInFrontEnd())
   n5 = GetOwnerFrontEndAnimInstance()
   > Set CustomFELobbyIdleFloat(CustomFELobbyIdleFloat:float=GetCurveValue(self=@n5, CurveName=FrontEnd_Idle_Bypass)) -> (Output_Get:float)
   > Set FE Backpack Pose Alpha(FE Backpack Pose Alpha:float=GetCurveValue(self=@n5, CurveName=Frontend_Backpack_Pose)) -> (Output_Get:float)
   Get bIsSkydiveFloating()
  ]}
]
variables [
 BeltMask Alpha {type=number}
 CustomFELobbyIdleFloat {type=number}
 FE Backpack Pose Alpha {type=number}
 FrontEnd_RigidBodiesDisabled {type=boolean}
 Gravity Override Params {type=GravityOverrideParamsStruct}
 GravityOverride {type=Vector}
 Is Skydive Floating Alpha {type=number}
]
