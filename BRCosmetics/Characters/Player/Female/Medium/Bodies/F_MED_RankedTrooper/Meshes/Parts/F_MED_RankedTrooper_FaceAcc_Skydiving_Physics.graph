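# Physics asset for the female Ranked Trooper's FaceAcc skydiving variant, covering a hooded cape and dangling belt strap.
# Three five-joint cape chains hang from a master on spine_05, with widening overlapping capsules, tight twist limits and free constraints tying side tips to center.
# A four-joint belt strap chain hangs from the pelvis with limited swing and twist.
# Capsules on torso, neck, upper arms and legs use free constraints.
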PhysicsAsset /BRCosmetics/Characters/Player/Female/Medium/Bodies/F_MED_RankedTrooper/Meshes/Parts/F_MED_RankedTrooper_FaceAcc_Skydiving_Physics
bodies {
 HoodedCape_bk_MSTR [{shapeType=Capsule, length=1, radius=1, shapeName=HoodedCape_bk_MSTR_capsule, rotation=<rot 0 -90 0>}]
 HoodedCape_bk_c_01_jnt [{shapeType=Capsule, length=18, radius=9, shapeName=HoodedCape_bk_c_01_jnt_capsule, center=<6 -7 0>}]
 HoodedCape_bk_c_02_jnt [{shapeType=Capsule, length=18, radius=10, shapeName=HoodedCape_bk_c_02_jnt_capsule, center=<6 -6 0>}]
 HoodedCape_bk_c_03_jnt [{shapeType=Capsule, length=18, radius=11, shapeName=HoodedCape_bk_c_03_jnt_capsule, center=<6 -6 0>}]
 HoodedCape_bk_c_04_jnt [{shapeType=Capsule, length=18, radius=12, shapeName=HoodedCape_bk_c_04_jnt_capsule, center=<6 -7 0>}]
 HoodedCape_bk_c_05_jnt [{shapeType=Capsule, length=18, radius=13, shapeName=HoodedCape_bk_c_05_jnt_capsule, center=<8 -8 0>}]
 HoodedCape_bk_l_01_jnt [{shapeType=Capsule, length=18, radius=9, shapeName=HoodedCape_bk_l_01_jnt_capsule, center=<6 -7 -5>}]
 HoodedCape_bk_l_02_jnt [{shapeType=Capsule, length=18, radius=10, shapeName=HoodedCape_bk_l_02_jnt_capsule, center=<6 -6 -5>}]
 HoodedCape_bk_l_03_jnt [{shapeType=Capsule, length=18, radius=11, shapeName=HoodedCape_bk_l_03_jnt_capsule, center=<6 -6 -5>}]
 HoodedCape_bk_l_04_jnt [{shapeType=Capsule, length=18, radius=12, shapeName=HoodedCape_bk_l_04_jnt_capsule, center=<6 -7 -5>}]
 HoodedCape_bk_l_05_jnt [{shapeType=Capsule, length=18, radius=13, shapeName=HoodedCape_bk_l_05_jnt_capsule, center=<8 -8 -5>}]
 HoodedCape_bk_r_01_jnt [{shapeType=Capsule, length=18, radius=9, shapeName=HoodedCape_bk_r_01_jnt_capsule, center=<6 -7 5>}]
 HoodedCape_bk_r_02_jnt [{shapeType=Capsule, length=18, radius=10, shapeName=HoodedCape_bk_r_02_jnt_capsule, center=<6 -6 5>}]
 HoodedCape_bk_r_03_jnt [{shapeType=Capsule, length=18, radius=11, shapeName=HoodedCape_bk_r_03_jnt_capsule, center=<6 -6 5>}]
 HoodedCape_bk_r_04_jnt [{shapeType=Capsule, length=18, radius=12, shapeName=HoodedCape_bk_r_04_jnt_capsule, center=<6 -7 5>}]
 HoodedCape_bk_r_05_jnt [{shapeType=Capsule, length=18, radius=13, shapeName=HoodedCape_bk_r_05_jnt_capsule, center=<8 -8 5>}]
 calf_l [{shapeType=Capsule, length=28, radius=8, shapeName=calf_l_capsule, center=<-22 0 0>, rotation=<rot 0 -90 90>}]
 calf_r [{shapeType=Capsule, length=28, radius=8, shapeName=calf_r_capsule, center=<22 0 0>, rotation=<rot 0 -90 90>}]
 clavicle_l [{shapeType=Capsule, length=12, radius=5.75, shapeName=clavicle_l_capsule, center=<3.5 -1 -1>, rotation=<rot 80 -0.0003 -0.0003>}]
 clavicle_r [{shapeType=Capsule, length=12, radius=5.75, shapeName=clavicle_r_capsule, center=<-3.5 1 1>, rotation=<rot 80 -0.0003 -0.0004>}]
 dyn_BeltStrap_01_jnt [{shapeType=Capsule, length=4, radius=1, shapeName=dyn_BeltStrap_01_jnt_capsule, center=<1.5 0 0>, rotation=<rot 0 0 90>}]
 dyn_BeltStrap_02_jnt [{shapeType=Capsule, length=4, radius=1, shapeName=dyn_BeltStrap_02_jnt_capsule, center=<1.5 0 0>, rotation=<rot 0 0 90>}]
 dyn_BeltStrap_03_jnt [{shapeType=Capsule, length=4, radius=1, shapeName=dyn_BeltStrap_03_jnt_capsule, center=<3 0 0>, rotation=<rot 0 0 90>}]
 dyn_BeltStrap_04_jnt [
  {shapeType=Capsule, length=14, radius=8, shapeName=dyn_BeltStrap_04_jnt_capsule, center=<16.1197 0 3>, rotation=<rot 0 90.0002 70>}
  {shapeType=Capsule, length=12.5, radius=8, shapeName=dyn_BeltStrap_04_jnt_capsule, center=<16.1197 0 3>, rotation=<rot 0 0 90>}
 ]
 dyn_BeltStrap_MSTR [{shapeType=Capsule, length=0, radius=0, shapeName=dyn_BeltStrap_MSTR_capsule}]
 foot_l [{shapeType=Capsule, length=10, radius=10, shapeName=foot_l_capsule, center=<0 8 0>, rotation=<rot 0 0 90>}]
 foot_r [{shapeType=Capsule, length=10, radius=10, shapeName=foot_r_capsule, center=<0 -8 0>, rotation=<rot 0 0 90>}]
 neck_01 [{shapeType=Capsule, length=8, radius=4.5, shapeName=neck_01_capsule, center=<3.6683 -0.2876 0>, rotation=<rot 90 0 0>}]
 neck_02 [{shapeType=Capsule, length=6, radius=4.5, shapeName=neck_02_capsule, center=<2 0 0>, rotation=<rot 0 90 90>}]
 pelvis [{shapeType=Capsule, length=10, radius=10, shapeName=pelvis_capsule, center=<0 -2.3096 0>, rotation=<rot 0 -90.0001 0>}]
 spine_01 [{shapeType=Capsule, length=8, radius=8, shapeName=spine_01_capsule, center=<0 -0.9904 0>, rotation=<rot 0 -90.0001 0>}]
 spine_02 [{shapeType=Capsule, length=8, radius=8, shapeName=spine_02_capsule, rotation=<rot 0 -90.0001 0>}]
 spine_03 [{shapeType=Capsule, length=15, radius=10, shapeName=spine_04_capsule, center=<0.229 2 0>, rotation=<rot 90 178.2101 -171.7897>}]
 spine_04 [{shapeType=Capsule, length=6, radius=12, shapeName=spine_04_capsule, rotation=<rot 0 89.9999 90>}]
 spine_05 [
  {shapeType=Capsule, length=12, radius=9.25, shapeName=spine_05_capsule, center=<-2.5963 0.209 0>, rotation=<rot 0 -0.0001 0>}
  {shapeType=Capsule, length=12, radius=8, shapeName=spine_05_capsule, center=<-5.4034 4.0924 0>}
 ]
 thigh_l [{shapeType=Capsule, length=28, radius=8, shapeName=thigh_l_capsule, center=<-22 0 0>, rotation=<rot 0 -90 90>}]
 thigh_r [{shapeType=Capsule, length=28, radius=8, shapeName=thigh_r_capsule, center=<22 0 0>, rotation=<rot 0 -90 90>}]
 upperarm_l [{shapeType=Capsule, length=20, radius=3.5, shapeName=upperarm_l_capsule, center=<12 -0.5 0>, rotation=<rot 90 0 0>}]
 upperarm_r [{shapeType=Capsule, length=20, radius=3.5, shapeName=upperarm_r_capsule, center=<-12 0.5 0>, rotation=<rot 90 0 0>}]
}
constraints [
 {
  bone1Name=spine_05
  bone2Name=spine_04
  swing1LimitDegrees=45
  swing1Motion=Free
  swing2LimitDegrees=45
  swing2Motion=Free
  twistLimitDegrees=45
  twistMotion=Free
 }
 {
  bone1Name=spine_01
  bone2Name=pelvis
  swing1LimitDegrees=45
  swing1Motion=Free
  swing2LimitDegrees=45
  swing2Motion=Free
  twistLimitDegrees=45
  twistMotion=Free
 }
 {
  bone1Name=spine_02
  bone2Name=spine_01
  swing1LimitDegrees=45
  swing1Motion=Free
  swing2LimitDegrees=45
  swing2Motion=Free
  twistLimitDegrees=45
  twistMotion=Free
 }
 {
  bone1Name=spine_03
  bone2Name=spine_02
  swing1LimitDegrees=45
  swing1Motion=Free
  swing2LimitDegrees=45
  swing2Motion=Free
  twistLimitDegrees=45
  twistMotion=Free
 }
 {
  bone1Name=spine_04
  bone2Name=spine_03
  swing1LimitDegrees=45
  swing1Motion=Free
  swing2LimitDegrees=45
  swing2Motion=Free
  twistLimitDegrees=45
  twistMotion=Free
 }
 {
  bone1Name=neck_01
  bone2Name=spine_05
  swing1LimitDegrees=45
  swing1Motion=Free
  swing2LimitDegrees=45
  swing2Motion=Free
  twistLimitDegrees=45
  twistMotion=Free
 }
 {
  bone1Name=clavicle_l
  bone2Name=spine_05
  swing1LimitDegrees=45
  swing1Motion=Free
  swing2LimitDegrees=45
  swing2Motion=Free
  twistLimitDegrees=45
  twistMotion=Free
 }
 {
  bone1Name=clavicle_r
  bone2Name=spine_05
  swing1LimitDegrees=45
  swing1Motion=Free
  swing2LimitDegrees=45
  swing2Motion=Free
  twistLimitDegrees=45
  twistMotion=Free
 }
 {
  bone1Name=upperarm_l
  bone2Name=clavicle_l
  swing1LimitDegrees=45
  swing1Motion=Free
  swing2LimitDegrees=45
  swing2Motion=Free
  twistLimitDegrees=45
  twistMotion=Free
 }
 {
  bone1Name=upperarm_r
  bone2Name=clavicle_r
  swing1LimitDegrees=45
  swing1Motion=Free
  swing2LimitDegrees=45
  swing2Motion=Free
  twistLimitDegrees=45
  twistMotion=Free
 }
 {
  bone1Name=neck_02
  bone2Name=neck_01
  swing1LimitDegrees=45
  swing1Motion=Free
  swing2LimitDegrees=45
  swing2Motion=Free
  twistLimitDegrees=45
  twistMotion=Free
 }
 {
  bone1Name=HoodedCape_bk_MSTR
  bone2Name=spine_05
  swing1LimitDegrees=45
  swing1Motion=Free
  swing2LimitDegrees=45
  swing2Motion=Free
  twistLimitDegrees=45
  twistMotion=Free
 }
 {bone1Name=HoodedCape_bk_c_01_jnt, bone2Name=HoodedCape_bk_MSTR, swing1LimitDegrees=35, swing2LimitDegrees=15, twistLimitDegrees=1}
 {bone1Name=HoodedCape_bk_c_02_jnt, bone2Name=HoodedCape_bk_c_01_jnt, swing1LimitDegrees=35, swing2LimitDegrees=15, twistLimitDegrees=1}
 {bone1Name=HoodedCape_bk_c_03_jnt, bone2Name=HoodedCape_bk_c_02_jnt, swing1LimitDegrees=35, swing2LimitDegrees=15, twistLimitDegrees=1}
 {bone1Name=HoodedCape_bk_c_04_jnt, bone2Name=HoodedCape_bk_c_03_jnt, swing1LimitDegrees=35, swing2LimitDegrees=15, twistLimitDegrees=1}
 {bone1Name=HoodedCape_bk_c_05_jnt, bone2Name=HoodedCape_bk_c_04_jnt, swing1LimitDegrees=35, swing2LimitDegrees=15, twistLimitDegrees=1}
 {bone1Name=HoodedCape_bk_l_01_jnt, bone2Name=HoodedCape_bk_MSTR, swing1LimitDegrees=35, swing2LimitDegrees=15, twistLimitDegrees=1}
 {bone1Name=HoodedCape_bk_l_02_jnt, bone2Name=HoodedCape_bk_l_01_jnt, swing1LimitDegrees=35, swing2LimitDegrees=15, twistLimitDegrees=1}
 {bone1Name=HoodedCape_bk_l_03_jnt, bone2Name=HoodedCape_bk_l_02_jnt, swing1LimitDegrees=35, swing2LimitDegrees=15, twistLimitDegrees=1}
 {bone1Name=HoodedCape_bk_l_04_jnt, bone2Name=HoodedCape_bk_l_03_jnt, swing1LimitDegrees=35, swing2LimitDegrees=15, twistLimitDegrees=1}
 {bone1Name=HoodedCape_bk_l_05_jnt, bone2Name=HoodedCape_bk_l_04_jnt, swing1LimitDegrees=35, swing2LimitDegrees=15, twistLimitDegrees=1}
 {bone1Name=HoodedCape_bk_r_01_jnt, bone2Name=HoodedCape_bk_MSTR, swing1LimitDegrees=35, swing2LimitDegrees=15, twistLimitDegrees=1}
 {bone1Name=HoodedCape_bk_r_02_jnt, bone2Name=HoodedCape_bk_r_01_jnt, swing1LimitDegrees=35, swing2LimitDegrees=15, twistLimitDegrees=1}
 {bone1Name=HoodedCape_bk_r_03_jnt, bone2Name=HoodedCape_bk_r_02_jnt, swing1LimitDegrees=35, swing2LimitDegrees=15, twistLimitDegrees=1}
 {bone1Name=HoodedCape_bk_r_04_jnt, bone2Name=HoodedCape_bk_r_03_jnt, swing1LimitDegrees=35, swing2LimitDegrees=15, twistLimitDegrees=1}
 {bone1Name=HoodedCape_bk_r_05_jnt, bone2Name=HoodedCape_bk_r_04_jnt, swing1LimitDegrees=35, swing2LimitDegrees=15, twistLimitDegrees=1}
 {
  bone1Name=foot_r
  bone2Name=calf_r
  swing1LimitDegrees=45
  swing1Motion=Free
  swing2LimitDegrees=45
  swing2Motion=Free
  twistLimitDegrees=45
  twistMotion=Free
 }
 {
  bone1Name=thigh_l
  bone2Name=pelvis
  swing1LimitDegrees=45
  swing1Motion=Free
  swing2LimitDegrees=45
  swing2Motion=Free
  twistLimitDegrees=45
  twistMotion=Free
 }
 {
  bone1Name=calf_l
  bone2Name=thigh_l
  swing1LimitDegrees=45
  swing1Motion=Free
  swing2LimitDegrees=45
  swing2Motion=Free
  twistLimitDegrees=45
  twistMotion=Free
 }
 {
  bone1Name=foot_l
  bone2Name=calf_l
  swing1LimitDegrees=45
  swing1Motion=Free
  swing2LimitDegrees=45
  swing2Motion=Free
  twistLimitDegrees=45
  twistMotion=Free
 }
 {
  bone1Name=thigh_r
  bone2Name=pelvis
  swing1LimitDegrees=45
  swing1Motion=Free
  swing2LimitDegrees=45
  swing2Motion=Free
  twistLimitDegrees=45
  twistMotion=Free
 }
 {
  bone1Name=calf_r
  bone2Name=thigh_r
  swing1LimitDegrees=45
  swing1Motion=Free
  swing2LimitDegrees=45
  swing2Motion=Free
  twistLimitDegrees=45
  twistMotion=Free
 }
 {bone1Name=dyn_BeltStrap_04_jnt, bone2Name=dyn_BeltStrap_03_jnt, swing1LimitDegrees=15, swing2LimitDegrees=25, twistLimitDegrees=5}
 {
  bone1Name=dyn_BeltStrap_MSTR
  bone2Name=pelvis
  swing1LimitDegrees=45
  swing1Motion=Free
  swing2LimitDegrees=45
  swing2Motion=Free
  twistLimitDegrees=45
  twistMotion=Free
 }
 {bone1Name=dyn_BeltStrap_01_jnt, bone2Name=dyn_BeltStrap_MSTR, swing1LimitDegrees=15, swing2LimitDegrees=25, twistLimitDegrees=5}
 {bone1Name=dyn_BeltStrap_02_jnt, bone2Name=dyn_BeltStrap_01_jnt, swing1LimitDegrees=15, swing2LimitDegrees=25, twistLimitDegrees=5}
 {bone1Name=dyn_BeltStrap_03_jnt, bone2Name=dyn_BeltStrap_02_jnt, swing1LimitDegrees=15, swing2LimitDegrees=25, twistLimitDegrees=5}
 {
  bone1Name=HoodedCape_bk_r_05_jnt
  bone2Name=HoodedCape_bk_c_05_jnt
  swing1LimitDegrees=45
  swing1Motion=Free
  swing2LimitDegrees=45
  swing2Motion=Free
  twistLimitDegrees=45
  twistMotion=Free
 }
 {
  bone1Name=HoodedCape_bk_l_05_jnt
  bone2Name=HoodedCape_bk_c_05_jnt
  swing1LimitDegrees=45
  swing1Motion=Free
  swing2LimitDegrees=45
  swing2Motion=Free
  twistLimitDegrees=45
  twistMotion=Free
 }
]
