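# Post-processes the female RankedTrooper face accessory mesh, animating its hair and hooded cape.
# Parent and body mesh poses are layered, then run through a Control Rig, cape offsets and hair AnimDynamics into a cached Presim pose.
# RigidBody cape branches switch physics assets for normal, skydiving or backbling states, the latter adding pelvis constraints and AnimDynamics.
# The EventGraph updates gravity override, movement, front-end and LOD-based alphas, and detects backbling on customization load.
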AnimBlueprint /BRCosmetics/Characters/Player/Female/Medium/Bodies/F_MED_RankedTrooper/Meshes/Parts/F_MED_RankedTrooper_FaceAcc_AnimBP
target_skeleton=&/F_MED_RankedTrooper_FaceAcc_Skeleton
graphs [
 AnimGraph {nodes [
   n1 = "Transform (Modify) Bone - Bone: HoodedCape_bk_c_02_jnt"(Alpha=1, Translation="0,0,0", Rotation="0,-15,0", Scale="1,1,1") {
    rotationMode=BMM_Additive
    rotationSpace=BCS_BoneSpace
    translationMode=BMM_Additive
    translationSpace=BCS_BoneSpace
    alphaBoolBlend {blendOption=Linear}
    alphaScaleBiasClamp {bInterpResult=true, interpSpeedDecreasing=3, interpSpeedIncreasing=3}
    boneToModify {boneName=HoodedCape_bk_c_02_jnt}
    rotation {yaw=-15}
   }
   n2 = "Transform (Modify) Bone - Bone: HoodedCape_bk_l_02_jnt"(ComponentPose=@n1, Alpha=1, Translation="0,0,0", Rotation="0,-15,0", Scale="1,1,1") {
    rotationMode=BMM_Additive
    rotationSpace=BCS_BoneSpace
    translationMode=BMM_Additive
    translationSpace=BCS_BoneSpace
    alphaBoolBlend {blendOption=Linear}
    alphaScaleBiasClamp {bInterpResult=true, interpSpeedDecreasing=3, interpSpeedIncreasing=3}
    boneToModify {boneName=HoodedCape_bk_l_02_jnt}
    rotation {yaw=-15}
   }
   "Transform (Modify) Bone - Bone: HoodedCape_bk_r_02_jnt"(ComponentPose=@n2, Alpha=1, Translation="0,0,0", Rotation="0,-15,0", Scale="1,1,1") {
    rotationMode=BMM_Additive
    rotationSpace=BCS_BoneSpace
    translationMode=BMM_Additive
    translationSpace=BCS_BoneSpace
    alphaBoolBlend {blendOption=Linear}
    alphaScaleBiasClamp {bInterpResult=true, interpSpeedDecreasing=3, interpSpeedIncreasing=3}
    boneToModify {boneName=HoodedCape_bk_r_02_jnt}
    rotation {yaw=-15}
   }
   n3 = "Transform (Modify) Bone - Bone: HoodedCape_bk_c_01_jnt"(ComponentPose=Local To Component(LocalPose=Use cached pose 'Presim'()), Alpha=1, Translation="0,-1,0", Rotation="0,-5,0", Scale="1,1,1") {
    rotationMode=BMM_Additive
    rotationSpace=BCS_BoneSpace
    translationMode=BMM_Additive
    translationSpace=BCS_BoneSpace
    alphaBoolBlend {blendOption=Linear}
    alphaScaleBiasClamp {bInterpResult=true, interpSpeedDecreasing=4, interpSpeedIncreasing=4}
    boneToModify {boneName=HoodedCape_bk_c_01_jnt}
    rotation {yaw=-5}
    translation {y=-1}
   }
   n4 = "Transform (Modify) Bone - Bone: HoodedCape_bk_l_01_jnt"(ComponentPose=@n3, Alpha=1, Translation="0,-1,0", Rotation="0,-5,0", Scale="1,1,1") {
    rotationMode=BMM_Additive
    rotationSpace=BCS_BoneSpace
    translationMode=BMM_Additive
    translationSpace=BCS_BoneSpace
    alphaBoolBlend {blendOption=Linear}
    alphaScaleBiasClamp {bInterpResult=true, interpSpeedDecreasing=4, interpSpeedIncreasing=4}
    boneToModify {boneName=HoodedCape_bk_l_01_jnt}
    rotation {yaw=-5}
    translation {y=-1}
   }
   n5 = "Transform (Modify) Bone - Bone: HoodedCape_bk_r_01_jnt"(ComponentPose=@n4, Alpha=1, Translation="0,-1,0", Rotation="0,-5,0", Scale="1,1,1") {
    rotationMode=BMM_Additive
    rotationSpace=BCS_BoneSpace
    translationMode=BMM_Additive
    translationSpace=BCS_BoneSpace
    alphaBoolBlend {blendOption=Linear}
    alphaScaleBiasClamp {bInterpResult=true, interpSpeedDecreasing=4, interpSpeedIncreasing=4}
    boneToModify {boneName=HoodedCape_bk_r_01_jnt}
    rotation {yaw=-5}
    translation {y=-1}
   }
   n6 = RigidBody!(ComponentPose=@n5, Alpha, OverrideWorldGravity:Vector=Get Gravity Override()) -> (Pose) {
    bClampLinearTranslationLimitToRefPose=true
    bOverrideWorldGravity=true
    overridePhysicsAsset=&/F_MED_RankedTrooper_FaceAcc_Physics
    simulationSpace=BaseBoneSpace
    alphaBoolBlend {blendOption=Linear}
    alphaScaleBiasClamp {bInterpResult=true, interpSpeedIncreasing=2}
    baseBoneRef {boneName=HoodedCape_bk_MSTR}
    componentAppliedLinearAccClamp=<3000 3000 3000>
    componentLinearAccScale=<0.5 0.7 0.7>
    componentLinearVelScale=<0.5 0.7 0.7>
    simSpaceSettings {
     maxAngularAcceleration=1000
     maxAngularVelocity=3
     maxLinearAcceleration=700
     maxLinearVelocity=500
     velocityScaleZ=0.5
     worldAlpha=1
     externalLinearDragV=<0.2 0.2>
    }
   }
   n7 = "Transform (Modify) Bone - Bone: HoodedCape_bk_c_01_jnt"(ComponentPose=Local To Component(LocalPose=Use cached pose 'Presim'()), Alpha=1, Translation="0,-5,0", Rotation="0,-20,0", Scale="1,1,1") {
    rotationMode=BMM_Additive
    rotationSpace=BCS_BoneSpace
    translationMode=BMM_Additive
    translationSpace=BCS_BoneSpace
    alphaBoolBlend {blendOption=Linear}
    alphaScaleBiasClamp {bInterpResult=true, interpSpeedDecreasing=4, interpSpeedIncreasing=4}
    boneToModify {boneName=HoodedCape_bk_c_01_jnt}
    rotation {yaw=-20}
    translation {y=-5}
   }
   n8 = "Transform (Modify) Bone - Bone: HoodedCape_bk_l_01_jnt"(ComponentPose=@n7, Alpha=1, Translation="0,-5,0", Rotation="0,-20,0", Scale="1,1,1") {
    rotationMode=BMM_Additive
    rotationSpace=BCS_BoneSpace
    translationMode=BMM_Additive
    translationSpace=BCS_BoneSpace
    alphaBoolBlend {blendOption=Linear}
    alphaScaleBiasClamp {bInterpResult=true, interpSpeedDecreasing=4, interpSpeedIncreasing=4}
    boneToModify {boneName=HoodedCape_bk_l_01_jnt}
    rotation {yaw=-20}
    translation {y=-5}
   }
   n9 = "Transform (Modify) Bone - Bone: HoodedCape_bk_r_01_jnt"(ComponentPose=@n8, Alpha=1, Translation="0,-5,0", Rotation="0,-20,0", Scale="1,1,1") {
    rotationMode=BMM_Additive
    rotationSpace=BCS_BoneSpace
    translationMode=BMM_Additive
    translationSpace=BCS_BoneSpace
    alphaBoolBlend {blendOption=Linear}
    alphaScaleBiasClamp {bInterpResult=true, interpSpeedDecreasing=4, interpSpeedIncreasing=4}
    boneToModify {boneName=HoodedCape_bk_r_01_jnt}
    rotation {yaw=-20}
    translation {y=-5}
   }
   n10 = RigidBody(ComponentPose=@n9, OverrideWorldGravity=Get Gravity Override()) {
    bClampLinearTranslationLimitToRefPose=true
    bOverrideWorldGravity=true
    overridePhysicsAsset=&/F_MED_RankedTrooper_FaceAcc_Skydiving_Physics
    simulationSpace=BaseBoneSpace
    alphaBoolBlend {blendOption=Linear}
    alphaScaleBiasClamp {bInterpResult=true, interpSpeedIncreasing=2}
    baseBoneRef {boneName=HoodedCape_bk_MSTR}
    componentAppliedLinearAccClamp=<3000 3000 3000>
    componentLinearAccScale=<0.5 0.7 0.7>
    componentLinearVelScale=<0.5 0.7 0.7>
    simSpaceSettings {
     maxAngularAcceleration=1000
     maxAngularVelocity=3
     maxLinearAcceleration=700
     maxLinearVelocity=500
     velocityScaleZ=0.5
     worldAlpha=1
     externalLinearDragV=<0.2 0.2>
    }
   }
   n11 = Blend Poses by bool(BlendPose_0=Component To Local(ComponentPose=@n10), BlendPose_1=Component To Local(ComponentPose=@n6), bActiveValue=Get bIsSkyDiving()) {blendType=Linear}
   n12 = "Transform (Modify) Bone - Bone: CapeShoulder_l_jnt"(ComponentPose=Local To Component(LocalPose=Use cached pose 'Presim'()), Alpha=1, Translation="0,0,2", Rotation="0,0,0", Scale="1,1,1") {
    rotationMode=BMM_Additive
    rotationSpace=BCS_BoneSpace
    translationMode=BMM_Additive
    translationSpace=BCS_BoneSpace
    alphaBoolBlend {blendOption=Linear}
    alphaScaleBiasClamp {bInterpResult=true, interpSpeedDecreasing=4, interpSpeedIncreasing=4}
    boneToModify {boneName=CapeShoulder_l_jnt}
    translation {z=2}
   }
   n13 = "Transform (Modify) Bone - Bone: CapeShoulder_r_jnt"(ComponentPose=@n12, Alpha=1, Translation="0,0,-2", Rotation="0,0,0", Scale="1,1,1") {
    rotationMode=BMM_Additive
    rotationSpace=BCS_BoneSpace
    translationMode=BMM_Additive
    translationSpace=BCS_BoneSpace
    alphaBoolBlend {blendOption=Linear}
    alphaScaleBiasClamp {bInterpResult=true, interpSpeedDecreasing=4, interpSpeedIncreasing=4}
    boneToModify {boneName=CapeShoulder_r_jnt}
    translation {z=-2}
   }
   n14 = "Transform (Modify) Bone - Bone: HoodedCape_bk_c_01_jnt"(ComponentPose=@n13, Alpha=1, Translation="0,-2.5,0", Rotation="0,0,0", Scale="1,1,1") {
    rotationMode=BMM_Additive
    rotationSpace=BCS_BoneSpace
    translationMode=BMM_Additive
    translationSpace=BCS_BoneSpace
    alphaBoolBlend {blendOption=Linear}
    alphaScaleBiasClamp {bInterpResult=true, interpSpeedDecreasing=4, interpSpeedIncreasing=4}
    boneToModify {boneName=HoodedCape_bk_c_01_jnt}
    translation {y=-2.5}
   }
   n15 = "Transform (Modify) Bone - Bone: HoodedCape_bk_l_01_jnt"(ComponentPose=@n14, Alpha=1, Translation="0,-2.5,0", Rotation="0,0,0", Scale="1,1,1") {
    rotationMode=BMM_Additive
    rotationSpace=BCS_BoneSpace
    translationMode=BMM_Additive
    translationSpace=BCS_BoneSpace
    alphaBoolBlend {blendOption=Linear}
    alphaScaleBiasClamp {bInterpResult=true, interpSpeedDecreasing=4, interpSpeedIncreasing=4}
    boneToModify {boneName=HoodedCape_bk_l_01_jnt}
    translation {y=-2.5}
   }
   n16 = "Transform (Modify) Bone - Bone: HoodedCape_bk_r_01_jnt"(ComponentPose=@n15, Alpha=1, Translation="0,-2.5,0", Rotation="0,0,0", Scale="1,1,1") {
    rotationMode=BMM_Additive
    rotationSpace=BCS_BoneSpace
    translationMode=BMM_Additive
    translationSpace=BCS_BoneSpace
    alphaBoolBlend {blendOption=Linear}
    alphaScaleBiasClamp {bInterpResult=true, interpSpeedDecreasing=4, interpSpeedIncreasing=4}
    boneToModify {boneName=HoodedCape_bk_r_01_jnt}
    translation {y=-2.5}
   }
   n17 = "Transform (Modify) Bone - Bone: HoodedCape_bk_c_02_jnt"(ComponentPose=@n16, Alpha=1, Translation="0,5,0", Rotation="0,-5,0", Scale="1,1,1") {
    rotationMode=BMM_Additive
    rotationSpace=BCS_BoneSpace
    translationMode=BMM_Additive
    translationSpace=BCS_BoneSpace
    alphaBoolBlend {blendOption=Linear}
    alphaScaleBiasClamp {bInterpResult=true, interpSpeedDecreasing=4, interpSpeedIncreasing=4}
    boneToModify {boneName=HoodedCape_bk_c_02_jnt}
    rotation {yaw=-5}
    translation {y=5}
   }
   n18 = "Transform (Modify) Bone - Bone: HoodedCape_bk_l_02_jnt"(ComponentPose=@n17, Alpha=1, Translation="0,5,0", Rotation="0,-5,0", Scale="1,1,1") {
    rotationMode=BMM_Additive
    rotationSpace=BCS_BoneSpace
    translationMode=BMM_Additive
    translationSpace=BCS_BoneSpace
    alphaBoolBlend {blendOption=Linear}
    alphaScaleBiasClamp {bInterpResult=true, interpSpeedDecreasing=4, interpSpeedIncreasing=4}
    boneToModify {boneName=HoodedCape_bk_l_02_jnt}
    rotation {yaw=-5}
    translation {y=5}
   }
   n19 = "Transform (Modify) Bone - Bone: HoodedCape_bk_r_02_jnt"(ComponentPose=@n18, Alpha=1, Translation="0,5,0", Rotation="0,-5,0", Scale="1,1,1") {
    rotationMode=BMM_Additive
    rotationSpace=BCS_BoneSpace
    translationMode=BMM_Additive
    translationSpace=BCS_BoneSpace
    alphaBoolBlend {blendOption=Linear}
    alphaScaleBiasClamp {bInterpResult=true, interpSpeedDecreasing=4, interpSpeedIncreasing=4}
    boneToModify {boneName=HoodedCape_bk_r_02_jnt}
    rotation {yaw=-5}
    translation {y=5}
   }
   n20 = "Transform (Modify) Bone - Bone: HoodedCape_bk_c_03_jnt"(ComponentPose=@n19, Alpha=1, Translation="0,-5,0", Rotation="0,-5,0", Scale="1,1,1") {
    rotationMode=BMM_Additive
    rotationSpace=BCS_BoneSpace
    translationMode=BMM_Additive
    translationSpace=BCS_BoneSpace
    alphaBoolBlend {blendOption=Linear}
    boneToModify {boneName=HoodedCape_bk_c_03_jnt}
    rotation {yaw=-5}
    translation {y=-5}
   }
   n21 = "Transform (Modify) Bone - Bone: HoodedCape_bk_l_03_jnt"(ComponentPose=@n20, Alpha=1, Translation="0,-5,0", Rotation="0,-5,0", Scale="1,1,1") {
    rotationMode=BMM_Additive
    rotationSpace=BCS_BoneSpace
    translationMode=BMM_Additive
    translationSpace=BCS_BoneSpace
    alphaBoolBlend {blendOption=Linear}
    boneToModify {boneName=HoodedCape_bk_l_03_jnt}
    rotation {yaw=-5}
    translation {y=-5}
   }
   n22 = "Transform (Modify) Bone - Bone: HoodedCape_bk_r_03_jnt"(ComponentPose=@n21, Alpha=1, Translation="0,-5,0", Rotation="0,-5,0", Scale="1,1,1") {
    rotationMode=BMM_Additive
    rotationSpace=BCS_BoneSpace
    translationMode=BMM_Additive
    translationSpace=BCS_BoneSpace
    alphaBoolBlend {blendOption=Linear}
    boneToModify {boneName=HoodedCape_bk_r_03_jnt}
    rotation {yaw=-5}
    translation {y=-5}
   }
   n23 = "Transform (Modify) Bone - Bone: HoodedCape_bk_c_03_jnt"(ComponentPose=@n22, Alpha=1, Translation="0,0,0", Rotation="0,-15,0", Scale="1,1,1") {
    rotationMode=BMM_Additive
    rotationSpace=BCS_BoneSpace
    translationMode=BMM_Additive
    translationSpace=BCS_BoneSpace
    alphaBoolBlend {blendOption=Linear}
    alphaScaleBiasClamp {bInterpResult=true, interpSpeedDecreasing=4, interpSpeedIncreasing=4}
    boneToModify {boneName=HoodedCape_bk_c_03_jnt}
    rotation {yaw=-15}
   }
   n24 = "Transform (Modify) Bone - Bone: HoodedCape_bk_l_03_jnt"(ComponentPose=@n23, Alpha=1, Translation="0,0,0", Rotation="0,-15,0", Scale="1,1,1") {
    rotationMode=BMM_Additive
    rotationSpace=BCS_BoneSpace
    translationMode=BMM_Additive
    translationSpace=BCS_BoneSpace
    alphaBoolBlend {blendOption=Linear}
    alphaScaleBiasClamp {bInterpResult=true, interpSpeedDecreasing=4, interpSpeedIncreasing=4}
    boneToModify {boneName=HoodedCape_bk_l_03_jnt}
    rotation {yaw=-15}
   }
   n25 = "Transform (Modify) Bone - Bone: HoodedCape_bk_r_03_jnt"(ComponentPose=@n24, Alpha=1, Translation="0,0,0", Rotation="0,-15,0", Scale="1,1,1") {
    rotationMode=BMM_Additive
    rotationSpace=BCS_BoneSpace
    translationMode=BMM_Additive
    translationSpace=BCS_BoneSpace
    alphaBoolBlend {blendOption=Linear}
    alphaScaleBiasClamp {bInterpResult=true, interpSpeedDecreasing=4, interpSpeedIncreasing=4}
    boneToModify {boneName=HoodedCape_bk_r_03_jnt}
    rotation {yaw=-15}
   }
   n26 = RigidBody(ComponentPose=@n25, OverrideWorldGravity=Get Gravity Override()) {
    bClampLinearTranslationLimitToRefPose=true
    bOverrideWorldGravity=true
    overridePhysicsAsset=&/F_MED_RankedTrooper_FaceAcc_Pin_Physics
    simulationSpace=BaseBoneSpace
    alphaBoolBlend {blendOption=Linear}
    alphaScaleBiasClamp {bInterpResult=true, interpSpeedIncreasing=2}
    baseBoneRef {boneName=HoodedCape_bk_MSTR}
    componentAppliedLinearAccClamp=<3000 3000 3000>
    componentLinearAccScale=<0.5 0.7 0.7>
    componentLinearVelScale=<0.5 0.7 0.7>
    simSpaceSettings {
     maxAngularAcceleration=1000
     maxAngularVelocity=3
     maxLinearAcceleration=700
     maxLinearVelocity=500
     velocityScaleZ=0.5
     worldAlpha=1
     externalLinearDragV=<0.2 0.2>
    }
   }
   n27 = "Constraint - HoodedCape_bk_c_03_jnt (P)"(ComponentPose=@n26, Alpha=1, ConstraintWeights_0=1) {
    alphaBoolBlend {blendOption=Linear}
    boneToModify {boneName=HoodedCape_bk_c_03_jnt}
    constraintSetup [{targetBone {boneName=pelvis}}]
    constraintWeights [1]
   }
   n28 = "Constraint - HoodedCape_bk_l_03_jnt (P)"(ComponentPose=@n27, Alpha=1, ConstraintWeights_0=1) {
    alphaBoolBlend {blendOption=Linear}
    boneToModify {boneName=HoodedCape_bk_l_03_jnt}
    constraintSetup [{targetBone {boneName=pelvis}}]
    constraintWeights [1]
   }
   n29 = "Constraint - HoodedCape_bk_r_03_jnt (P)"(ComponentPose=@n28, Alpha=1, ConstraintWeights_0=1) {
    alphaBoolBlend {blendOption=Linear}
    boneToModify {boneName=HoodedCape_bk_r_03_jnt}
    constraintSetup [{targetBone {boneName=pelvis}}]
    constraintWeights [1]
   }
   n30 = Anim Dynamics - Bone: HoodedCape_bk_c_03_jnt!(ComponentPose=@n29, Alpha=1, GravityOverride:Vector="0,0,-980", ExternalForce="0,0,0") -> (Pose) {
    angularDampingOverride=0.9999
    angularSpringConstant=50
    bAngularSpring=true
    bOverrideAngularDamping=true
    bUseGravityOverride=true
    bUsePlanarLimit=false
    gravityScale=0.35
    simulationSpace=RootRelative
    alphaBoolBlend {blendOption=Linear}
    boundBone {boneName=HoodedCape_bk_c_03_jnt}
    componentAppliedLinearAccClamp=<6000 6000 6000>
    componentLinearAccScale=<0.5 0.7 0.7>
    componentLinearVelScale=<0.5 0.7 0.7>
    gravityOverride {z=-980}
    physicsBodyDefinitions [{
      boundBone {boneName=HoodedCape_bk_c_03_jnt}
      boxExtents=<20 30 30>
      constraintSetup {angularLimitsMax=<0 6 0>, angularLimitsMin=<0 -6 -25>, angularTarget=<1 0 0>}
      localJointOffset=<10 0 0>
     }]
   }
   n31 = Anim Dynamics - Bone: HoodedCape_bk_l_03_jnt!(ComponentPose=@n30, Alpha=1, GravityOverride:Vector="0,0,-980", ExternalForce="0,0,0") -> (Pose) {
    angularDampingOverride=0.9999
    angularSpringConstant=50
    bAngularSpring=true
    bOverrideAngularDamping=true
    bUseGravityOverride=true
    bUsePlanarLimit=false
    gravityScale=0.35
    simulationSpace=RootRelative
    alphaBoolBlend {blendOption=Linear}
    boundBone {boneName=HoodedCape_bk_l_03_jnt}
    componentAppliedLinearAccClamp=<6000 6000 6000>
    componentLinearAccScale=<0.5 0.7 0.7>
    componentLinearVelScale=<0.5 0.7 0.7>
    gravityOverride {z=-980}
    physicsBodyDefinitions [{
      boundBone {boneName=HoodedCape_bk_l_03_jnt}
      boxExtents=<20 30 30>
      constraintSetup {angularLimitsMax=<0 6 0>, angularLimitsMin=<0 -6 -25>, angularTarget=<1 0 0>}
      localJointOffset=<10 0 0>
     }]
   }
   n32 = Anim Dynamics - Bone: HoodedCape_bk_r_03_jnt!(ComponentPose=@n31, Alpha=1, GravityOverride:Vector="0,0,-980", ExternalForce="0,0,0") -> (Pose) {
    angularDampingOverride=0.9999
    angularSpringConstant=50
    bAngularSpring=true
    bOverrideAngularDamping=true
    bUseGravityOverride=true
    bUsePlanarLimit=false
    gravityScale=0.35
    simulationSpace=RootRelative
    alphaBoolBlend {blendOption=Linear}
    boundBone {boneName=HoodedCape_bk_r_03_jnt}
    componentAppliedLinearAccClamp=<6000 6000 6000>
    componentLinearAccScale=<0.5 0.7 0.7>
    componentLinearVelScale=<0.5 0.7 0.7>
    gravityOverride {z=-980}
    physicsBodyDefinitions [{
      boundBone {boneName=HoodedCape_bk_r_03_jnt}
      boxExtents=<20 30 30>
      constraintSetup {angularLimitsMax=<0 6 0>, angularLimitsMin=<0 -6 -25>, angularTarget=<1 0 0>}
      localJointOffset=<10 0 0>
     }]
   }
   n33 = "Constraint - HoodedCape_bk_c_03_jnt (P)"(ComponentPose=@n32, Alpha=1, ConstraintWeights_0=1) {
    alphaBoolBlend {blendOption=Linear}
    alphaScaleBiasClamp {bInterpResult=true, interpSpeedDecreasing=4}
    boneToModify {boneName=HoodedCape_bk_c_03_jnt}
    constraintSetup [{targetBone {boneName=pelvis}}]
    constraintWeights [1]
   }
   n34 = "Constraint - HoodedCape_bk_l_03_jnt (P)"(ComponentPose=@n33, Alpha=1, ConstraintWeights_0=1) {
    alphaBoolBlend {blendOption=Linear}
    alphaScaleBiasClamp {bInterpResult=true, interpSpeedDecreasing=4}
    boneToModify {boneName=HoodedCape_bk_l_03_jnt}
    constraintSetup [{targetBone {boneName=pelvis}}]
    constraintWeights [1]
   }
   n35 = "Constraint - HoodedCape_bk_r_03_jnt (P)"(ComponentPose=@n34, Alpha=1, ConstraintWeights_0=1) {
    alphaBoolBlend {blendOption=Linear}
    alphaScaleBiasClamp {bInterpResult=true, interpSpeedDecreasing=4}
    boneToModify {boneName=HoodedCape_bk_r_03_jnt}
    constraintSetup [{targetBone {boneName=pelvis}}]
    constraintWeights [1]
   }
   n36 = Anim Dynamics - Bone: HoodedCape_bk_c_03_jnt(ComponentPose=@n35, Alpha=1, GravityOverride="0,0,-980", ExternalForce="0,0,0") {
    angularDampingOverride=0.9999
    angularSpringConstant=50
    bAngularSpring=true
    bOverrideAngularDamping=true
    bUseGravityOverride=true
    bUsePlanarLimit=false
    gravityScale=0.35
    simulationSpace=RootRelative
    alphaBoolBlend {blendOption=Linear}
    alphaScaleBiasClamp {bInterpResult=true, interpSpeedDecreasing=4}
    boundBone {boneName=HoodedCape_bk_c_03_jnt}
    componentAppliedLinearAccClamp=<6000 6000 6000>
    componentLinearAccScale=<0.5 0.7 0.7>
    componentLinearVelScale=<0.5 0.7 0.7>
    gravityOverride {z=-980}
    physicsBodyDefinitions [{
      boundBone {boneName=HoodedCape_bk_c_03_jnt}
      boxExtents=<20 30 30>
      constraintSetup {angularLimitsMax=<0 6 0>, angularLimitsMin=<0 -6 -25>, angularTarget=<1 0 0>}
      localJointOffset=<10 0 0>
     }]
   }
   n37 = Anim Dynamics - Bone: HoodedCape_bk_l_03_jnt(ComponentPose=@n36, Alpha=1, GravityOverride="0,0,-980", ExternalForce="0,0,0") {
    angularDampingOverride=0.9999
    angularSpringConstant=50
    bAngularSpring=true
    bOverrideAngularDamping=true
    bUseGravityOverride=true
    bUsePlanarLimit=false
    gravityScale=0.35
    simulationSpace=RootRelative
    alphaBoolBlend {blendOption=Linear}
    alphaScaleBiasClamp {bInterpResult=true, interpSpeedDecreasing=4}
    boundBone {boneName=HoodedCape_bk_l_03_jnt}
    componentAppliedLinearAccClamp=<6000 6000 6000>
    componentLinearAccScale=<0.5 0.7 0.7>
    componentLinearVelScale=<0.5 0.7 0.7>
    gravityOverride {z=-980}
    physicsBodyDefinitions [{
      boundBone {boneName=HoodedCape_bk_l_03_jnt}
      boxExtents=<20 30 30>
      constraintSetup {angularLimitsMax=<0 6 0>, angularLimitsMin=<0 -6 -25>, angularTarget=<1 0 0>}
      localJointOffset=<10 0 0>
     }]
   }
   n38 = Anim Dynamics - Bone: HoodedCape_bk_r_03_jnt(ComponentPose=@n37, Alpha=1, GravityOverride="0,0,-980", ExternalForce="0,0,0") {
    angularDampingOverride=0.9999
    angularSpringConstant=50
    bAngularSpring=true
    bOverrideAngularDamping=true
    bUseGravityOverride=true
    bUsePlanarLimit=false
    gravityScale=0.35
    simulationSpace=RootRelative
    alphaBoolBlend {blendOption=Linear}
    alphaScaleBiasClamp {bInterpResult=true, interpSpeedDecreasing=4}
    boundBone {boneName=HoodedCape_bk_r_03_jnt}
    componentAppliedLinearAccClamp=<6000 6000 6000>
    componentLinearAccScale=<0.5 0.7 0.7>
    componentLinearVelScale=<0.5 0.7 0.7>
    gravityOverride {z=-980}
    physicsBodyDefinitions [{
      boundBone {boneName=HoodedCape_bk_r_03_jnt}
      boxExtents=<20 30 30>
      constraintSetup {angularLimitsMax=<0 6 0>, angularLimitsMin=<0 -6 -25>, angularTarget=<1 0 0>}
      localJointOffset=<10 0 0>
     }]
   }
   n39 = Component To Local(ComponentPose=@n38)
   n40 = Get Is Wearing Backbling()
   n41 = "Transform (Modify) Bone - Bone: cape_adjustment_jnt"(
    ComponentPose=Local To Component(LocalPose=Blend Poses by bool(BlendPose_0=@n39, BlendPose_1=@n11, bActiveValue=@n40) {blendType=Linear})
    Alpha=1
    Translation="0,0,0"
    Rotation="0,20,0"
    Scale="1,1,1"
   ) {
    rotationMode=BMM_Additive
    rotationSpace=BCS_BoneSpace
    translationMode=BMM_Additive
    translationSpace=BCS_BoneSpace
    alphaBoolBlend {blendOption=Linear}
    alphaScaleBiasClamp {bInterpResult=true, interpSpeedDecreasing=4, interpSpeedIncreasing=6}
    boneToModify {boneName=cape_adjustment_jnt}
    rotation {yaw=20}
   }
   n42 = "Transform (Modify) Bone - Bone: cape_adjustment_jnt"(ComponentPose=@n41, Alpha=1, Translation="0,0,0", Rotation="20,0,0", Scale="1,1,1") {
    rotationMode=BMM_Additive
    rotationSpace=BCS_BoneSpace
    translationMode=BMM_Additive
    translationSpace=BCS_BoneSpace
    alphaBoolBlend {blendOption=Linear}
    alphaScaleBiasClamp {bInterpResult=true, interpSpeedDecreasing=4, interpSpeedIncreasing=6}
    boneToModify {boneName=cape_adjustment_jnt}
    rotation {pitch=20}
   }
   Output Pose(Result=Component To Local(ComponentPose=@n42)) {layerGroup=None}
   n43 = Layered blend per bone(
    BasePose=Copy Pose From Mesh() {bCopyCurves=true, bUseAttachedParent=true}
    BlendPoses_0=Copy Pose From Mesh(SourceMeshComponent=Get body Mesh() -> (body Mesh:Skeletal Mesh Component Object Reference))
   ) {layerSetup [{branchFilters [{boneName=upperarm_l}, {boneName=upperarm_r}, {boneName=thigh_l}, {boneName=thigh_r}, {boneName=dyn_BeltStrap_MSTR}]}]}
   n44 = Control Rig(Source=@n43) {
    alphaBoolBlend {blendOption=Linear}
    controlRigAssetReference {blueprintRigClass=&/F_MED_RankedTrooper_FaceAcc_CtrlRig.F_MED_RankedTrooper_FaceAcc_CtrlRig_C}
   }
   n45 = "Transform (Modify) Bone - Bone: HoodedCape_bk_c_01_jnt"(ComponentPose=Local To Component(LocalPose=@n44), Alpha=1, Translation="0,0,0", Rotation="0,5,0", Scale="1,1,1") {
    rotationMode=BMM_Additive
    rotationSpace=BCS_BoneSpace
    translationMode=BMM_Additive
    translationSpace=BCS_BoneSpace
    alphaBoolBlend {blendOption=Linear}
    boneToModify {boneName=HoodedCape_bk_c_01_jnt}
    rotation {yaw=5}
   }
   n46 = "Transform (Modify) Bone - Bone: HoodedCape_bk_l_01_jnt"(ComponentPose=@n45, Alpha=1, Translation="0,0,0", Rotation="-5,5,0", Scale="1,1,1") {
    rotationMode=BMM_Additive
    rotationSpace=BCS_BoneSpace
    translationMode=BMM_Additive
    translationSpace=BCS_BoneSpace
    alphaBoolBlend {blendOption=Linear}
    boneToModify {boneName=HoodedCape_bk_l_01_jnt}
    rotation {pitch=-5, yaw=5}
   }
   n47 = "Transform (Modify) Bone - Bone: HoodedCape_bk_r_01_jnt"(ComponentPose=@n46, Alpha=1, Translation="0,0,0", Rotation="5,5,0", Scale="1,1,1") {
    rotationMode=BMM_Additive
    rotationSpace=BCS_BoneSpace
    translationMode=BMM_Additive
    translationSpace=BCS_BoneSpace
    alphaBoolBlend {blendOption=Linear}
    boneToModify {boneName=HoodedCape_bk_r_01_jnt}
    rotation {pitch=5, yaw=5}
   }
   n48 = "Transform (Modify) Bone - Bone: HoodedCape_bk_c_02_jnt"(ComponentPose=@n47, Alpha=1, Translation="0,2,0", Rotation="0,0,0", Scale="1,1,1") {
    rotationMode=BMM_Additive
    rotationSpace=BCS_BoneSpace
    translationMode=BMM_Additive
    translationSpace=BCS_BoneSpace
    alphaBoolBlend {blendOption=Linear}
    boneToModify {boneName=HoodedCape_bk_c_02_jnt}
    translation {y=2}
   }
   n49 = "Transform (Modify) Bone - Bone: HoodedCape_bk_l_02_jnt"(ComponentPose=@n48, Alpha=1, Translation="0,2,0", Rotation="5,0,0", Scale="1,1,1") {
    rotationMode=BMM_Additive
    rotationSpace=BCS_BoneSpace
    translationMode=BMM_Additive
    translationSpace=BCS_BoneSpace
    alphaBoolBlend {blendOption=Linear}
    boneToModify {boneName=HoodedCape_bk_l_02_jnt}
    rotation {pitch=5}
    translation {y=2}
   }
   n50 = "Transform (Modify) Bone - Bone: HoodedCape_bk_r_02_jnt"(ComponentPose=@n49, Alpha=1, Translation="0,2,0", Rotation="-5,0,0", Scale="1,1,1") {
    rotationMode=BMM_Additive
    rotationSpace=BCS_BoneSpace
    translationMode=BMM_Additive
    translationSpace=BCS_BoneSpace
    alphaBoolBlend {blendOption=Linear}
    boneToModify {boneName=HoodedCape_bk_r_02_jnt}
    rotation {pitch=-5}
    translation {y=2}
   }
   n51 = "Transform (Modify) Bone - Bone: HoodedCape_bk_c_03_jnt"(ComponentPose=@n50, Alpha=1, Translation="0,-2,0", Rotation="0,-10,0", Scale="1,1,1") {
    rotationMode=BMM_Additive
    rotationSpace=BCS_BoneSpace
    translationMode=BMM_Additive
    translationSpace=BCS_BoneSpace
    alphaBoolBlend {blendOption=Linear}
    boneToModify {boneName=HoodedCape_bk_c_03_jnt}
    rotation {yaw=-10}
    translation {y=-2}
   }
   n52 = "Transform (Modify) Bone - Bone: HoodedCape_bk_l_03_jnt"(ComponentPose=@n51, Alpha=1, Translation="0,-2,0", Rotation="5,-10,0", Scale="1,1,1") {
    rotationMode=BMM_Additive
    rotationSpace=BCS_BoneSpace
    translationMode=BMM_Additive
    translationSpace=BCS_BoneSpace
    alphaBoolBlend {blendOption=Linear}
    boneToModify {boneName=HoodedCape_bk_l_03_jnt}
    rotation {pitch=5, yaw=-10}
    translation {y=-2}
   }
   n53 = "Transform (Modify) Bone - Bone: HoodedCape_bk_r_03_jnt"(ComponentPose=@n52, Alpha=1, Translation="0,-2,0", Rotation="-5,-10,0", Scale="1,1,1") {
    rotationMode=BMM_Additive
    rotationSpace=BCS_BoneSpace
    translationMode=BMM_Additive
    translationSpace=BCS_BoneSpace
    alphaBoolBlend {blendOption=Linear}
    boneToModify {boneName=HoodedCape_bk_r_03_jnt}
    rotation {pitch=-5, yaw=-10}
    translation {y=-2}
   }
   n54 = Anim Dynamics - Bone: dyn_L_HairBang_03_jnt(ComponentPose=@n53, Alpha=1, ExternalForce="0,0,0") {
    angularDampingOverride=0.9999
    angularSpringConstant=200
    bAngularSpring=true
    bOverrideAngularDamping=true
    bUseGravityOverride=true
    bUsePlanarLimit=false
    gravityScale=0.35
    simulationSpace=RootRelative
    alphaBoolBlend {blendOption=Linear}
    boundBone {boneName=dyn_L_HairBang_03_jnt}
    componentAppliedLinearAccClamp=<6000 6000 6000>
    componentLinearAccScale=<0.5 0.7 0.7>
    componentLinearVelScale=<0.5 0.7 0.7>
    gravityOverride {z=-980}
    physicsBodyDefinitions [{
      boundBone {boneName=dyn_L_HairBang_03_jnt}
      boxExtents=<20 30 30>
      constraintSetup {angularLimitsMax=<0 0 5>, angularLimitsMin=<0 -10 -5>, angularTarget=<1 0 0>}
      localJointOffset=<6 0 0>
     }]
   }
   n55 = Anim Dynamics - Bone: dyn_L_HairTip_01_jnt(ComponentPose=@n54, Alpha=1, ExternalForce="0,0,0") {
    angularDampingOverride=0.9999
    angularSpringConstant=200
    bAngularSpring=true
    bOverrideAngularDamping=true
    bUseGravityOverride=true
    bUsePlanarLimit=false
    gravityScale=0.35
    simulationSpace=RootRelative
    alphaBoolBlend {blendOption=Linear}
    boundBone {boneName=dyn_L_HairTip_01_jnt}
    componentAppliedLinearAccClamp=<6000 6000 6000>
    componentLinearAccScale=<0.5 0.7 0.7>
    componentLinearVelScale=<0.5 0.7 0.7>
    gravityOverride {z=-980}
    physicsBodyDefinitions [{
      boundBone {boneName=dyn_L_HairTip_01_jnt}
      boxExtents=<20 30 30>
      constraintSetup {angularLimitsMax=<0 0 5>, angularLimitsMin=<0 -10 -5>, angularTarget=<1 0 0>}
      localJointOffset=<6 0 0>
     }]
   }
   n56 = Anim Dynamics - Bone: dyn_L_HairTip_02_jnt(ComponentPose=@n55, Alpha=1, ExternalForce="0,0,0") {
    angularDampingOverride=0.9999
    angularSpringConstant=200
    bAngularSpring=true
    bOverrideAngularDamping=true
    bUseGravityOverride=true
    bUsePlanarLimit=false
    gravityScale=0.35
    simulationSpace=RootRelative
    alphaBoolBlend {blendOption=Linear}
    boundBone {boneName=dyn_L_HairTip_02_jnt}
    componentAppliedLinearAccClamp=<6000 6000 6000>
    componentLinearAccScale=<0.5 0.7 0.7>
    componentLinearVelScale=<0.5 0.7 0.7>
    gravityOverride {z=-980}
    physicsBodyDefinitions [{
      boundBone {boneName=dyn_L_HairTip_02_jnt}
      boxExtents=<20 30 30>
      constraintSetup {angularLimitsMax=<0 0 5>, angularLimitsMin=<0 -10 -5>, angularTarget=<1 0 0>}
      localJointOffset=<6 0 0>
     }]
   }
   n57 = Anim Dynamics - Bone: dyn_R_HairBang_03_jnt(ComponentPose=@n56, Alpha=1, ExternalForce="0,0,0") {
    angularDampingOverride=0.9999
    angularSpringConstant=200
    bAngularSpring=true
    bOverrideAngularDamping=true
    bUseGravityOverride=true
    bUsePlanarLimit=false
    gravityScale=0.35
    simulationSpace=RootRelative
    alphaBoolBlend {blendOption=Linear}
    boundBone {boneName=dyn_R_HairBang_03_jnt}
    componentAppliedLinearAccClamp=<6000 6000 6000>
    componentLinearAccScale=<0.5 0.7 0.7>
    componentLinearVelScale=<0.5 0.7 0.7>
    gravityOverride {z=-980}
    physicsBodyDefinitions [{
      boundBone {boneName=dyn_R_HairBang_03_jnt}
      boxExtents=<20 30 30>
      constraintSetup {angularLimitsMax=<0 0 5>, angularLimitsMin=<0 -10 -5>, angularTarget=<1 0 0>}
      localJointOffset=<6 0 0>
     }]
   }
   n58 = Anim Dynamics - Bone: dyn_R_HairBang_04_jnt(ComponentPose=@n57, Alpha=1, ExternalForce="0,0,0") {
    angularDampingOverride=0.9999
    angularSpringConstant=200
    bAngularSpring=true
    bOverrideAngularDamping=true
    bUseGravityOverride=true
    bUsePlanarLimit=false
    gravityScale=0.35
    simulationSpace=RootRelative
    alphaBoolBlend {blendOption=Linear}
    boundBone {boneName=dyn_R_HairBang_04_jnt}
    componentAppliedLinearAccClamp=<6000 6000 6000>
    componentLinearAccScale=<0.5 0.7 0.7>
    componentLinearVelScale=<0.5 0.7 0.7>
    gravityOverride {z=-980}
    physicsBodyDefinitions [{
      boundBone {boneName=dyn_R_HairBang_04_jnt}
      boxExtents=<20 30 30>
      constraintSetup {angularLimitsMax=<0 0 5>, angularLimitsMin=<0 -10 -5>, angularTarget=<1 0 0>}
      localJointOffset=<6 0 0>
     }]
   }
   n59 = Anim Dynamics - Bone: dyn_R_Hairstrand_03_jnt(ComponentPose=@n58, Alpha=1, ExternalForce="0,0,0") {
    angularDampingOverride=0.9999
    angularSpringConstant=200
    bAngularSpring=true
    bOverrideAngularDamping=true
    bUseGravityOverride=true
    bUsePlanarLimit=false
    gravityScale=0.35
    simulationSpace=RootRelative
    alphaBoolBlend {blendOption=Linear}
    boundBone {boneName=dyn_R_Hairstrand_03_jnt}
    componentAppliedLinearAccClamp=<6000 6000 6000>
    componentLinearAccScale=<0.5 0.7 0.7>
    componentLinearVelScale=<0.5 0.7 0.7>
    gravityOverride {z=-980}
    physicsBodyDefinitions [{
      boundBone {boneName=dyn_R_Hairstrand_03_jnt}
      boxExtents=<20 30 30>
      constraintSetup {angularLimitsMax=<0 0 5>, angularLimitsMin=<0 -10 -5>, angularTarget=<1 0 0>}
      localJointOffset=<6 0 0>
     }]
   }
   n60 = Anim Dynamics - Bone: dyn_R_Hairstrand_04_jnt(ComponentPose=@n59, Alpha=1, ExternalForce="0,0,0") {
    angularDampingOverride=0.9999
    angularSpringConstant=200
    bAngularSpring=true
    bOverrideAngularDamping=true
    bUseGravityOverride=true
    bUsePlanarLimit=false
    gravityScale=0.35
    simulationSpace=RootRelative
    alphaBoolBlend {blendOption=Linear}
    boundBone {boneName=dyn_R_Hairstrand_04_jnt}
    componentAppliedLinearAccClamp=<6000 6000 6000>
    componentLinearAccScale=<0.5 0.7 0.7>
    componentLinearVelScale=<0.5 0.7 0.7>
    gravityOverride {z=-980}
    physicsBodyDefinitions [{
      boundBone {boneName=dyn_R_Hairstrand_04_jnt}
      boxExtents=<20 30 30>
      constraintSetup {angularLimitsMax=<0 0 5>, angularLimitsMin=<0 -10 -5>, angularTarget=<1 0 0>}
      localJointOffset=<6 0 0>
     }]
   }
   n61 = Anim Dynamics - Bone: dyn_R_SmallHairBang_02_jnt(ComponentPose=@n60, Alpha=1, ExternalForce="0,0,0") {
    angularDampingOverride=0.9999
    angularSpringConstant=200
    bAngularSpring=true
    bOverrideAngularDamping=true
    bUseGravityOverride=true
    bUsePlanarLimit=false
    gravityScale=0.35
    simulationSpace=RootRelative
    alphaBoolBlend {blendOption=Linear}
    boundBone {boneName=dyn_R_SmallHairBang_02_jnt}
    componentAppliedLinearAccClamp=<6000 6000 6000>
    componentLinearAccScale=<0.5 0.7 0.7>
    componentLinearVelScale=<0.5 0.7 0.7>
    gravityOverride {z=-980}
    physicsBodyDefinitions [{
      boundBone {boneName=dyn_R_SmallHairBang_02_jnt}
      boxExtents=<20 30 30>
      constraintSetup {angularLimitsMax=<0 0 5>, angularLimitsMin=<0 -10 -5>, angularTarget=<1 0 0>}
      localJointOffset=<6 0 0>
     }]
   }
   n62 = Anim Dynamics - Bone: dyn_R_SmallHairBang_03_jnt(ComponentPose=@n61, Alpha=1, ExternalForce="0,0,0") {
    angularDampingOverride=0.9999
    angularSpringConstant=200
    bAngularSpring=true
    bOverrideAngularDamping=true
    bUseGravityOverride=true
    bUsePlanarLimit=false
    gravityScale=0.35
    simulationSpace=RootRelative
    alphaBoolBlend {blendOption=Linear}
    boundBone {boneName=dyn_R_SmallHairBang_03_jnt}
    componentAppliedLinearAccClamp=<6000 6000 6000>
    componentLinearAccScale=<0.5 0.7 0.7>
    componentLinearVelScale=<0.5 0.7 0.7>
    gravityOverride {z=-980}
    physicsBodyDefinitions [{
      boundBone {boneName=dyn_R_SmallHairBang_03_jnt}
      boxExtents=<20 30 30>
      constraintSetup {angularLimitsMax=<0 0 5>, angularLimitsMin=<0 -10 -5>, angularTarget=<1 0 0>}
      localJointOffset=<6 0 0>
     }]
   }
   n63 = Anim Dynamics - Bone: dyn_R_SmallHairBang_04_jnt(ComponentPose=@n62, Alpha=1, ExternalForce="0,0,0") {
    angularDampingOverride=0.9999
    angularSpringConstant=200
    bAngularSpring=true
    bOverrideAngularDamping=true
    bUseGravityOverride=true
    bUsePlanarLimit=false
    gravityScale=0.35
    simulationSpace=RootRelative
    alphaBoolBlend {blendOption=Linear}
    boundBone {boneName=dyn_R_SmallHairBang_04_jnt}
    componentAppliedLinearAccClamp=<6000 6000 6000>
    componentLinearAccScale=<0.5 0.7 0.7>
    componentLinearVelScale=<0.5 0.7 0.7>
    gravityOverride {z=-980}
    physicsBodyDefinitions [{
      boundBone {boneName=dyn_R_SmallHairBang_04_jnt}
      boxExtents=<20 30 30>
      constraintSetup {angularLimitsMax=<0 0 5>, angularLimitsMin=<0 -10 -5>, angularTarget=<1 0 0>}
      localJointOffset=<6 0 0>
     }]
   }
   Save cached pose 'Presim'(Pose=Component To Local(ComponentPose=@n63)) {cacheName=Presim}
  ]}
 EventGraph {nodes [
   Event BlueprintUpdateAnimation()
   > n1 = Sequence!(execute) -> (then_0, then_1, then_2:Exec)
   > n2 = Gravity Override State Function DEPRECATED(CharacterPartAnimInstance=Self-Reference(), GameTimeInSeconds=GetGameTimeInSeconds(), GravityOverrideParams=Get Gravity Override Params())
   > Set Gravity Override(Gravity Override=@n2)
   Set Is Skydiving Alpha(execute=@n1.then_2, Is Skydiving Alpha:double="To Float (Boolean)"(InBool=Get bIsSkyDiving())) -> (Output_Get:double)
   > Set Is Skydive Floating Alpha(Is Skydive Floating Alpha:double="To Float (Boolean)"(InBool=Get bIsSkydiveFloating())) -> (Output_Get:double)
   n3 = NOT Boolean(A=Get bIsCrouching())
   n4 = Get bIsMoving2D()
   > Set Is moving not Crouching Alpha(Is moving not Crouching Alpha:double="To Float (Boolean)"(InBool=AND Boolean(A=@n4, B=@n3))) -> (Output_Get:double)
   n5 = Get bIsMoving2D()
   n6 = Get bRigidBodyEnabled()
   > Set Is Moving 2D RbanDisabled(Is Moving 2D RbanDisabled:double="To Float (Boolean)"(InBool=AND Boolean(A=@n5, B=@n6))) -> (Output_Get:double)
   > Set FE_Idle_Alpha(FE_Idle_Alpha:double=GetCurveValue(self=GetOwnerFrontEndAnimInstance(), CurveName=FrontEnd_Idle_Bypass)) -> (Output_Get:double)
   n7 = Get CurrentLOD()
   > n8 = Branch(Condition=NOT Boolean(A="Equal (Integer)"(A=@n7, B=1)))
   > Set Cape Alpha(Cape Alpha:double=1) -> (Output_Get:double)
   Set Cape Alpha(execute=@n8.else, Cape Alpha=0.5)
   n9 = Event OnCharacterCustomizationLoaded()
   > n10 = Sequence()
   n11 = Get Array Item with EFort Custom Part Type(Array=@n9.CustomMeshParts)
   > Set body Mesh(body Mesh:Skeletal Mesh Component Object Reference=@n11) -> (Output_Get:Skeletal Mesh Component Object Reference)
   n12 = CheckForBackbling()
   Set Is Wearing Backbling(execute=@n10.then_1, Is Wearing Backbling=OR Boolean(A=@n12, B=@n12.bIsBackblingLong))
   Event BlueprintInitializeAnimation()
   > Set body Mesh(body Mesh=@n11)
  ]}
]
variables [
 body Mesh {type=SkeletalMeshComponent}
 Cape Alpha {type=number}
 FE_Idle_Alpha {type=number}
 Gravity Override {type=Vector}
 Gravity Override Params {type=GravityOverrideParamsStruct}
 Is Moving 2D RbanDisabled {type=number}
 Is moving not Crouching Alpha {type=number}
 Is Skydive Floating Alpha {type=number}
 Is Skydiving Alpha {type=number}
 Is Wearing Backbling {type=boolean}
]
